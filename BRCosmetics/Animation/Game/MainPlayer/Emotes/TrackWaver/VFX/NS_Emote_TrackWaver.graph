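# Niagara system for the TrackWaver emote, layering warm golden sparks, overhead cones, mist and twinkling motes over a roughly five-second timeline.
# Ambient_Hand_Sparks samples the left-hand skeletal mesh and drifts sparks with curl noise and vortex forces.
# ConesMistCondensed packs many layers into one emitter, using spawn-group switches for each layer's look and a visibility tag.
# Its opacity adapts to camera angle, time of day, platform and front-end flags.
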NiagaraSystem /BRCosmetics/Animation/Game/MainPlayer/Emotes/TrackWaver/VFX/NS_Emote_TrackWaver
dependencies {
 usedDataInterfaces [
  &NiagaraDataInterfaceArrayFloat
  &NiagaraDataInterfaceArrayFloat2
  &NiagaraDataInterfaceArrayInt32
  &NiagaraDataInterfaceCurve
  &NiagaraDataInterfaceColorCurve
  &NiagaraDataInterfaceVectorCurve
  &NiagaraDataInterfaceArrayColor
  &NiagaraDataInterfacePlatformSet
  &NiagaraDataInterfaceCamera
  &NiagaraDataInterfaceSpriteRendererInfo
  &NiagaraDataInterfaceSkeletalMesh
 ]
 usedDynamicInputs []
 usedModules [
  &self:SystemSpawnScript.SystemScriptSource.SystemScriptGraph.NiagaraNodeAssignment_0.SetVariables_#1
  &self:SystemSpawnScript.SystemScriptSource.SystemScriptGraph.NiagaraNodeAssignment_2.SetVariables_#2
  &SystemState
 ]
}
emitters {
 Ambient_Hand_Sparks {
  emitterUpdateScript [
   &EmitterState(Life Cycle Mode=::NewEnumerator1, Loop Behavior=::NewEnumerator1, Loop Duration=5.1, UseLoopDelay=false)
   &SpawnRate(
    SpawnRate=&Multiply_Float(
     A=50
     B=&FloatFromCurve(
      FloatCurve=DataInterface() {propertyValues {Curve {keys [
          <at 2.0952 0>
          <at 2.3184 1>
          {arriveTangent=-0.8602, interpMode=RCIM_Cubic, leaveTangent=-0.8602, time=3.091, value=1}
          {arriveTangent=0.6515, interpMode=RCIM_Cubic, leaveTangent=0.6515, tangentMode=RCTM_User, time=3.4385, value=0.0364}
          {arriveTangent=-0.0447, interpMode=RCIM_Cubic, leaveTangent=-0.0447, tangentMode=RCTM_User, time=3.5176, value=2}
          {arriveTangent=0.0789, interpMode=RCIM_Cubic, leaveTangent=0.0789, tangentMode=RCTM_User, time=3.8002, value=0.0114}
         ]}}}
      CurveIndex=$Emitter.Age
      Scale Curve=&FloatFromCurve(
       FloatCurve=DataInterface() {propertyValues {Curve {defaultValue=1, keys [{tangentMode=RCTM_User, time=0, value=0}, {tangentMode=RCTM_User, time=1, value=1}]}}}
       CurveIndex=$Emitter.NormalizedLoopAge
      )
     )
    )
    Use Spawn Probability=false
   )
  ]
  particleSpawnScript [
   &InitializeParticle(
    Color Maximum=<rgba 10 6.7471 0 1>
    Color Minimum=<rgba 5 3.8356 1.3697 1>
    Color Mode=::NewEnumerator2
    Lifetime Min=0.6
    Lifetime Mode=::NewEnumerator1
    Mass Mode=::NewEnumerator3
    Ribbon Width=8
    Ribbon Width Mode=::NewEnumerator1
    Sprite Rotation Mode=::NewEnumerator2
    Sprite Size Mode=::NewEnumerator1
    Uniform Sprite Size Max=3.5
    Uniform Sprite Size Min=0.75
    UsePositionOffset=false
    Write Material Random=true
   )
   &SkeletalMeshLocation(
    MaskSpawnGroup=false
    Mesh Sampling Type=::NewEnumerator5
    Sampled Position Offset=<0 0 5>
    Sampled Position Offset Coordinate Space:ENiagaraSkelMeshTransforms=::NewEnumerator2
    Skeletal Mesh=$System.left hand
    Socket Sampling Mode=::NewEnumerator5
   )
   &ShapeLocation(MaskSpawnGroup=false, Override Seed=false, Sphere Radius=8, UseOwnerScale=true)
  ]
  particleUpdateScript [
   &ParticleState()
   &ScaleSpriteSize(Uniform Curve Sprite Scale=DataInterface() {propertyValues {Curve {
       defaultValue=0
       keys [
        {arriveTangent=3.6109, interpMode=RCIM_Cubic, leaveTangent=3.6109, tangentMode=RCTM_User, time=0, value=1}
        {arriveTangent=-1, interpMode=RCIM_Cubic, leaveTangent=-1, time=0.1543, value=1.2569}
        {interpMode=RCIM_Cubic, tangentMode=RCTM_User, time=1, value=0}
       ]
      }}})
   &GravityForce(Gravity=&RandomRangeVector(Minimum=<0 0 10>, Maximum=<0 0 30>))
   &ScaleColor(
    Scale Alpha=&Multiply_Float(
     A=&FloatFromCurve(FloatCurve=DataInterface() {propertyValues {Curve {keys [<at 5 1>,<at 5.1 0>]}}}, CurveIndex=$Emitter.Age)
     B=&FloatFromCurve(
      FloatCurve=DataInterface() {propertyValues {Curve {
         defaultValue=1.0113
         keys [
          {interpMode=RCIM_Cubic, tangentMode=RCTM_User, time=0, value=0}
          {interpMode=RCIM_Cubic, tangentMode=RCTM_User, time=0.212, value=1}
          {tangentMode=RCTM_User, time=1, value=0}
         ]
        }}}
      Scale Curve=&FloatFromCurve(FloatCurve=DataInterface() {propertyValues {Curve {keys [
           <at 0 0.5 RCIM_Cubic>
           {arriveTangent=-0.1247, interpMode=RCIM_Cubic, leaveTangent=-0.1247, time=0.1795, value=1.0667}
           {arriveTangent=0.0434, interpMode=RCIM_Cubic, leaveTangent=0.0434, time=0.4011, value=0.45}
           {arriveTangent=0.1899, interpMode=RCIM_Cubic, leaveTangent=0.1899, time=0.5636, value=1.0834}
           {arriveTangent=0.206, interpMode=RCIM_Cubic, leaveTangent=0.206, time=0.7439, value=0.5151}
           {arriveTangent=-0.0589, interpMode=RCIM_Cubic, leaveTangent=-0.0589, time=0.8503, value=1.1424}
           <at 1 0.5 RCIM_Cubic>
          ]}}})
     )
    )
    Scale RGB=&MakeVectorFromLinearColorRGB(Color=&ColorFromCurve(ColorCurve=DataInterface() {propertyValues {
        AlphaCurve {keys []}
        BlueCurve {keys [
          <at 0 1>
          <at 0.1193 3.479>
          <at 0.1681 9>
          <at 0.214 5>
          {arriveTangent=-16.5932, interpMode=RCIM_Cubic, leaveTangent=-16.5932, time=0.4249, value=1}
          <at 0.455 1>
          <at 1 1 RCIM_Cubic>
         ]}
        GreenCurve {keys [
          <at 0 1>
          <at 0.1193 3.479>
          <at 0.1681 9>
          <at 0.214 5>
          {arriveTangent=-16.5932, interpMode=RCIM_Cubic, leaveTangent=-16.5932, time=0.4249, value=1}
          <at 0.455 1>
          <at 1 1 RCIM_Cubic>
         ]}
        RedCurve {keys [
          <at 0 1>
          <at 0.1193 3.479>
          <at 0.1681 9>
          <at 0.214 5>
          {arriveTangent=-16.5932, interpMode=RCIM_Cubic, leaveTangent=-16.5932, time=0.4249, value=1}
          <at 0.455 1>
          <at 1 1 RCIM_Cubic>
         ]}
       }}))
    ScaleA=true
    ScaleRGB=true
   )
   &CurlNoiseForce(
    Bias Noise=false
    Noise Strength=&FloatFromCurve(
     FloatCurve=DataInterface() {propertyValues {Curve {
        defaultValue=1
        keys [{arriveTangent=3.585, interpMode=RCIM_Cubic, leaveTangent=3.585, tangentMode=RCTM_User, time=0, value=1}, <at 1.0057 1.7345 RCIM_Cubic>]
       }}}
     Scale Curve=&RandomRangeFloat(Minimum=100, Maximum=200)
    )
    Pan Noise=false
   )
   &Drag(Drag=&RandomRangeFloat(Minimum=2, Maximum=8), Use Linear Drag=true, Use Rotational Drag=false)
   &VortexForce(UseOriginOffset=false)
   &SolveForcesAndVelocity(Clamp Velocity=false, Limit Acceleration=false)
   &ScaleSpriteSizeBySpeed(Max Scale Factor=<1 1.5>, Min Scale Factor=<0.5 1>)
  ]
  rendererClasses [&NiagaraSpriteRendererProperties]
  renderers [{rendererClass=&NiagaraSpriteRendererProperties}]
 }
 ConesMistCondensed {
  emitterUpdateScript [
   &EmitterState(Life Cycle Mode=::NewEnumerator1, Loop Behavior=::NewEnumerator1, Loop Duration=5.1, UseLoopDelay=false)
   /SetVariables(Emitter.Glow:LinearColor=&RandomRangeLinearColor(Minimum=<rgba 11 7.9259 0.9473 1>, Maximum=<rgba 7 3.8513 0.2697 1>)) {isSetParameters=true}
   &SpawnBurst_Instantaneous(Spawn Count=2, Spawn Time=0.3, Use Spawn Probability=false, UseLoopCountLimit=false)
   &SpawnRate(SpawnRate=3.5, Use Spawn Probability=false)
   &SpawnBurst_Instantaneous(Spawn Group=1, Use Spawn Probability=false, UseLoopCountLimit=false) {module=SpawnBurst_Instantaneous001}
   &SpawnRate(
    SpawnGroup=2
    SpawnRate=&Multiply_Float(
     A=9
     B=&FloatFromCurve(
      FloatCurve=DataInterface() {propertyValues {Curve {keys [
          <at 0 0 RCIM_Cubic>
          {arriveTangent=1.0847, interpMode=RCIM_Cubic, leaveTangent=1.0847, time=0.4732, value=0.2616}
          {arriveTangentWeight=0.1523, interpMode=RCIM_Cubic, leaveTangentWeight=1.1627, tangentMode=RCTM_User, time=0.9219, value=1}
          {arriveTangentWeight=1.1562, interpMode=RCIM_Cubic, leaveTangentWeight=0.0974, tangentMode=RCTM_User, time=3.7369, value=1}
          {arriveTangent=-0.795, interpMode=RCIM_Cubic, leaveTangent=-0.795, time=4.2923, value=0.3332}
          <at 4.9948 0 RCIM_Cubic>
         ]}}}
      CurveIndex=$Emitter.Age
     )
    )
    Use Spawn Probability=false
   ) {module=SpawnRate001}
   &SpawnRate(
    SpawnGroup=3
    SpawnRate=&Multiply_Float(
     A=14
     B=&FloatFromCurve(
      FloatCurve=DataInterface() {propertyValues {Curve {keys [<at 0.0532 0>,<at 0.7301 1>,<at 2.8289 0.7436>,<at 3.6946 0>]}}}
      CurveIndex=$Emitter.Age
     )
    )
    Use Spawn Probability=false
   ) {module=SpawnRate002}
   &SpawnRate(
    SpawnGroup=4
    SpawnRate=&Multiply_Float(
     A=16.5
     B=&FloatFromCurve(
      FloatCurve=DataInterface() {propertyValues {Curve {keys [<at 0.0532 0>,<at 0.7301 1>,<at 1.9907 1.35>,<at 3.1228 0.7436>,<at 3.6946 0>]}}}
      CurveIndex=$Emitter.Age
     )
    )
    Use Spawn Probability=false
   ) {module=SpawnRate003}
   &SpawnRate(
    SpawnGroup=5
    SpawnRate=&Multiply_Float(
     A=16.5
     B=&FloatFromCurve(
      FloatCurve=DataInterface() {propertyValues {Curve {keys [<at 0.0532 0>,<at 0.7301 1>,<at 1.9907 1.35>,<at 3.1228 0.7436>,<at 3.6946 0>]}}}
      CurveIndex=$Emitter.Age
     )
    )
    Use Spawn Probability=false
   ) {module=SpawnRate004}
   &SpawnRate(
    SpawnGroup=6
    SpawnRate=&Multiply_Float(
     A=13
     B=&FloatFromCurve(
      FloatCurve=DataInterface() {propertyValues {Curve {keys [
          <at 0 0 RCIM_Cubic>
          {arriveTangent=1.0847, interpMode=RCIM_Cubic, leaveTangent=1.0847, time=0.4732, value=0.2616}
          {arriveTangentWeight=0.1523, interpMode=RCIM_Cubic, leaveTangentWeight=1.1627, tangentMode=RCTM_User, time=0.9219, value=1}
          {arriveTangentWeight=1.1562, interpMode=RCIM_Cubic, leaveTangentWeight=0.0974, tangentMode=RCTM_User, time=3.7369, value=1}
          {arriveTangent=-0.795, interpMode=RCIM_Cubic, leaveTangent=-0.795, time=4.2923, value=0.3332}
          <at 4.9948 0 RCIM_Cubic>
         ]}}}
      CurveIndex=$Emitter.Age
     )
    )
    Use Spawn Probability=false
   ) {module=SpawnRate005}
   &SpawnBurst_Instantaneous(Spawn Group=7, Spawn Time=0.5, Use Spawn Probability=false, UseLoopCountLimit=false) {module=SpawnBurst_Instantaneous002}
   &SpawnBurst_Instantaneous(Spawn Group=8, Use Spawn Probability=false, UseLoopCountLimit=false) {module=SpawnBurst_Instantaneous003}
   &SpawnBurst_Instantaneous(Spawn Group=9, Use Spawn Probability=false, UseLoopCountLimit=false) {module=SpawnBurst_Instantaneous004}
  ]
  particleSpawnScript [
   /SetVariables(Particles.VisibilityTag=&SpawnGroup()) {isSetParameters=true}
   &InitializeParticle(
    Color=&MergeHelper_ColorSwitch(
     0=&RandomRangeLinearColor(Minimum=<rgba 1.25 0.7118 0.0635 1>, Maximum=<rgba 1 0.7907 0.3032 1>)
     1=<rgba 1.5 1.0726 0.0537 1>
     2=&RandomRangeLinearColor(Minimum=<rgba 1 0.7302 0.2238 0.1>, Maximum=<rgba 1 0.6873 0.063 0.33>)
     3=&SelectLinearColorFromArray(
      Array Sampling Mode=::NewEnumerator0
      Direct Array Index=?
      Color Selection Array=DataInterface() {propertyValues {ColorData [
         <rgba 1 0.3057 0.0798 0.6348>
         <rgba 1 0.5772 0.0154 0.9026>
         <rgba 1 0.8515 0.2635 1>
         <rgba 1 0.8181 0.4984 1>
         <rgba 1 0.3223 0.0071 0.5687>
         <rgba 1 0.4495 0.2541 0.8678>
        ]}}
      Random Seed=0
     )
     4=<rgba 1 1 1 1>
     5=<rgba 1 1 1 1>
     6=&RandomRangeLinearColor(
      Minimum=&RandomRangeLinearColor(Minimum=<rgba 1.5 1.1929 0.5801 0.7913>, Maximum=<rgba 1 0.7844 0.289 0.7043>)
      Maximum=<rgba 1 0.6875 0.107 0.8296>
     )
     7=<rgba 1 0.8714 0.4965 1>
     8=<rgba 1 1 1 1>
     9=<rgba 1 1 1 1>
     Index=&SpawnGroup()
     Num=::NewEnumerator8
    )
    Lifetime=&MergeHelper_FloatSwitch(
     0=&RandomRangeFloat(Minimum=2, Maximum=3)
     1=5.1
     2=&RandomRangeFloat(Minimum=2, Maximum=3)
     3=&RandomRangeFloat(Minimum=1.5, Maximum=3)
     4=&RandomRangeFloat(Minimum=1.5, Maximum=3)
     5=&RandomRangeFloat(Minimum=1.5, Maximum=3)
     6=&RandomRangeFloat(Minimum=1.4, Maximum=1.75)
     7=2.1
     8=5.1
     9=5.1
     Index=&SpawnGroup()
     Num=::NewEnumerator8
    )
    Mass=&MergeHelper_FloatSwitch(
     0=1
     1=1
     2=&RandomRangeFloat(Minimum=0.75, Maximum=1.25)
     3=&RandomRangeFloat(Minimum=0.75, Maximum=1.25)
     4=&RandomRangeFloat(Minimum=0.75, Maximum=1.25)
     5=&RandomRangeFloat(Minimum=0.75, Maximum=1.25)
     6=&RandomRangeFloat(Minimum=0.75, Maximum=1.25)
     7=1
     Index=&SpawnGroup()
     Num=::NewEnumerator6
    )
    Mass Mode=::NewEnumerator1
    Mesh Scale=&MergeHelper_VectorSwitch(
     0=&RandomRangeVector(Minimum=<1 1 0.7>, Maximum=<1.6 1.6 1.2>)
     1=<2.2 2.2 1.25>
     2=?
     3=?
     Index=&SpawnGroup()
     Num=::NewEnumerator0
    )
    Mesh Scale Mode=::NewEnumerator4
    Position Offset=&MergeHelper_VectorSwitch(
     0=<0 7 270>
     1=<0 7 275>
     2=<0 0 0>
     3=<0 0 0>
     4=<0 0 0>
     5=<0 0 0>
     6=<0 0 0>
     7=<0 22 165>
     8=<0 0 150>
     9=<0 70 170>
     Index=&SpawnGroup()
     Num=::NewEnumerator8
    )
    Sprite Rotation Angle=&MaskFloat(
     Mask Source=0
     Mask Parameter=?
     Mask=7
     Masked Float=0
     Passthrough Float=&RandomRangeFloat(Minimum=360, Maximum=-360)
    )
    Sprite Rotation Mode=::NewEnumerator0
    Sprite Size=&MergeHelper_Vector2Switch(
     0=<0 0>
     1=<0 0>
     2=&Vector2DFromFloat(Value=&RandomRangeFloat(Minimum=160, Maximum=280))
     3=&Vector2DFromFloat(Value=&SelectFloatFromArray(
       Array Sampling Mode=::NewEnumerator0
       Direct Array Index=?
       Float Selection Array=DataInterface() {propertyValues {FloatData [2,4,10,6,20,30,40,3,5]}}
       Random Seed=0
      ))
     4=&Vector2DFromFloat(Value=&UniformAOrBFloat(
       A=&RandomRangeFloat(Minimum=3, Maximum=5)
       B=&RandomRangeFloat(Minimum=2, Maximum=3)
       Evaluation Type=::NewEnumerator0
       Recalculate Random Each Loop=true
       Override Seed=false
      ))
     5=&Vector2DFromFloat(Value=&UniformAOrBFloat(
       A=&RandomRangeFloat(Minimum=3, Maximum=5)
       B=&RandomRangeFloat(Minimum=2, Maximum=3)
       Evaluation Type=::NewEnumerator0
       Recalculate Random Each Loop=true
       Override Seed=false
      ))
     6=&Vector2DFromFloat(Value=&RandomRangeFloat(Minimum=30, Maximum=80))
     7=<133 160>
     Index=&SpawnGroup()
     Num=::NewEnumerator6
    )
    Sprite Size Mode=::NewEnumerator4
    UsePositionOffset=true
    Write Material Random=true
   )
   &InitialMeshOrientation(
    Mesh Orientation Mode=::NewEnumerator2
    Rotation=&RandomRangeVector(Minimum=<0 0 0>, Maximum=<0 0 1>)
    Use Rotation Vector=true
   )
   &CameraOffset(Camera Offset Amount=&MergeHelper_FloatSwitch(
     0=-100
     1=&MakeCustomFloatFromBool(
      Bool=&SetBoolByFloatComparison(
       A=&FN_PickFloatDependOnPlatform(
        TargetPlatform=DataInterface() {propertyValues {Platforms {
           qualityLevelMask=-29
           deviceProfileStates [
            {profileName=Linux, qualityLevelMask=1, setQualityLevelMask=3}
            {profileName=Mac, qualityLevelMask=1, setQualityLevelMask=3}
            {profileName=Mobile, qualityLevelMask=4294967295, setQualityLevelMask=4294967295}
            {profileName=Windows, qualityLevelMask=1, setQualityLevelMask=3}
            {profileName=Switch, qualityLevelMask=4294967295, setQualityLevelMask=4294967295}
           ]
          }}}
        TargetPlatformActive=1
        TargetPlatformNoActive=0
       )
       B=0
      )
      True Float=-200
      False Float=-100
     )
     2=&RandomRangeFloat(Minimum=-13, Maximum=-20)
     3=&RandomRangeFloat(Minimum=-10, Maximum=-20)
     4=&RandomRangeFloat(Minimum=-10, Maximum=-20)
     5=&RandomRangeFloat(Minimum=-10, Maximum=-20)
     6=&RandomRangeFloat(Minimum=-15, Maximum=-30)
     7=-25
     Index=&SpawnGroup()
     Num=::NewEnumerator6
    ))
   &ShapeLocation(
    MaskSpawnGroup=false
    Non Uniform Scale=&MergeHelper_VectorSwitch(
     0=<0 0 0>
     1=<0 0 0>
     2=<1 1 1.5>
     3=<1 1 1.6>
     4=<0.75 0.75 1.6>
     5=<0.75 0.75 1.6>
     6=<0.7 0.7 1.7>
     7=<0 0 0>
     Index=&SpawnGroup()
     Num=::NewEnumerator6
    )
    Offset=&MergeHelper_VectorSwitch(
     0=<0 0 0>
     1=<0 0 0>
     2=&RandomRangeVector(Minimum=<0 7 150>, Maximum=<0 7 210>)
     3=<0 7 190>
     4=<0 7 190>
     5=<0 7 190>
     6=<0 0 180>
     7=<0 0 0>
     Index=&SpawnGroup()
     Num=::NewEnumerator6
    )
    Offset Mode=::NewEnumerator0
    Override Seed=false
    Sphere Radius=&MergeHelper_FloatSwitch(0=0, 1=0, 2=40, 3=60, 4=60, 5=60, 6=50, 7=0, Index=&SpawnGroup(), Num=::NewEnumerator6)
    UseOwnerScale=true
   )
   &AddVelocity(Velocity=&MergeHelper_VectorSwitch(
     0=<0 0 0>
     1=<0 0 0>
     2=<0 5 -5>
     3=<0 0 0>
     4=<0 0 0>
     5=<0 0 0>
     6=&RandomRangeVector(Minimum=<0 0 -20>, Maximum=<0 0 -45>)
     7=<0 0 0>
     Index=&SpawnGroup()
     Num=::NewEnumerator6
    ))
   &ApplyInitialForces()
  ]
  particleUpdateScript [
   &ParticleState()
   &ScaleSpriteSize(
    Scale Sprite Size Mode=::NewEnumerator0
    Uniform Scale Factor=&MergeHelper_FloatSwitch(
     0=0
     1=0
     2=&FloatFromCurve(FloatCurve=DataInterface() {propertyValues {Curve {
         defaultValue=0
         keys [
          {arriveTangent=1.1215, interpMode=RCIM_Cubic, leaveTangent=1.1215, tangentMode=RCTM_User, time=0.0073, value=0.1553}
          {interpMode=RCIM_Cubic, tangentMode=RCTM_User, time=0.4524, value=1}
          {arriveTangent=1.3255, tangentMode=RCTM_User, time=1.0037, value=1.3591}
         ]
        }}})
     3=1
     4=&FloatFromCurve(
      FloatCurve=DataInterface() {propertyValues {Curve {
         defaultValue=0
         keys [
          {interpMode=RCIM_Cubic, tangentMode=RCTM_User, time=0, value=0}
          {interpMode=RCIM_Cubic, tangentMode=RCTM_User, time=0.5, value=1}
          {tangentMode=RCTM_User, time=1, value=0}
         ]
        }}}
      Scale Curve=&RandomRangeFloat(Minimum=1, Maximum=1.6)
     )
     5=&FloatFromCurve(
      FloatCurve=DataInterface() {propertyValues {Curve {
         defaultValue=0
         keys [
          {interpMode=RCIM_Cubic, tangentMode=RCTM_User, time=0, value=0}
          {interpMode=RCIM_Cubic, tangentMode=RCTM_User, time=0.5, value=1}
          {tangentMode=RCTM_User, time=1, value=0}
         ]
        }}}
      Scale Curve=&RandomRangeFloat(Minimum=1, Maximum=1.6)
     )
     6=1
     7=&FloatFromCurve(FloatCurve=DataInterface() {propertyValues {Curve {
         defaultValue=1
         keys [
          {arriveTangent=5.4233, interpMode=RCIM_Cubic, leaveTangent=5.4233, tangentMode=RCTM_User, time=0, value=0.25}
          {arriveTangent=1, interpMode=RCIM_Cubic, leaveTangent=1, time=0.28, value=1.03}
          <at 1 1.25 RCIM_Cubic>
         ]
        }}})
     Index=&SpawnGroup()
     Num=::NewEnumerator6
    )
   )
   &ScaleColor(
    Scale Alpha=&MergeHelper_FloatSwitch(
     0=&Lerp_Float(
      A=&Multiply_Float(
       A=&FloatFromCurve(
        FloatCurve=DataInterface() {propertyValues {Curve {keys [
            <at 0 0 RCIM_Cubic>
            {arriveTangentWeight=0.0513, interpMode=RCIM_Cubic, leaveTangentWeight=1.414, tangentMode=RCTM_User, time=0.3423, value=1}
            {arriveTangentWeight=1.388, interpMode=RCIM_Cubic, leaveTangentWeight=0.0968, tangentMode=RCTM_User, time=2.2793, value=1}
            {arriveTangent=-0.5309, interpMode=RCIM_Cubic, leaveTangent=-0.5309, time=2.9875, value=0.3129}
            <at 4.163 0 RCIM_Cubic>
           ]}}}
        CurveIndex=$Emitter.Age
        Scale Curve=&MakeCustomFloatFromBool(Bool=&Boolean_OrOperation(Bool1=$User.bIsFrontEnd, Bool2=$User.bIsFrontEndPreview), True Float=0.5, False Float=1)
       )
       B=&FloatFromCurve(
        FloatCurve=DataInterface() {propertyValues {Curve {keys [<at 0 0 RCIM_Cubic>,<at 0.4 1 RCIM_Cubic>,<at 1 0 RCIM_Cubic>]}}}
        Scale Curve=&FNTimeOfDayScalar(Night Scale=0.1, Morning Scale=0.25, Day Scale=0.5, Evening Scale=0.25)
       )
      )
      B=0
      Alpha=&ClampFloat(Float=&RemapRange(
        Input Value To Remap=&DotProduct(Vector A=&ReturnCameraVector(Camera Vectors=::NewEnumerator1, Coordinate Space=::World), Vector B=<0 0 -1>)
        Input Min=0.7
        Input Max=0.9
        Output Min=0
        Output Max=1
       ))
     )
     1=1
     2=&Multiply_Float(
      A=&FloatFromCurve(
       FloatCurve=DataInterface() {propertyValues {Curve {keys [
           <at 0 0 RCIM_Cubic>
           {arriveTangent=1.0847, interpMode=RCIM_Cubic, leaveTangent=1.0847, time=0.4732, value=0.2616}
           {arriveTangentWeight=0.1523, interpMode=RCIM_Cubic, leaveTangentWeight=1.1627, tangentMode=RCTM_User, time=0.9219, value=1}
           {arriveTangentWeight=1.1562, interpMode=RCIM_Cubic, leaveTangentWeight=0.0974, tangentMode=RCTM_User, time=3.5554, value=1}
           {arriveTangent=-0.6947, interpMode=RCIM_Cubic, leaveTangent=-0.6947, time=4.1722, value=0.4047}
           <at 4.9948 0 RCIM_Cubic>
          ]}}}
       CurveIndex=$Emitter.Age
       Scale Curve=&Multiply_Float(
        A=&MakeCustomFloatFromBool(Bool=$User.bIsFrontEndPreview, True Float=0.7, False Float=1)
        B=&Multiply_Float(
         A=&FNTimeOfDayScalar(Night Scale=0.8, Morning Scale=1, Evening Scale=1)
         B=&FN_PickFloatDependOnPlatform(TargetPlatformActive=1, TargetPlatformNoActive=0.6)
        )
       )
      )
      B=&RandomRangeFloat(
       Minimum=&FloatFromCurve(
        FloatCurve=DataInterface() {propertyValues {Curve {keys [
            <at -0.0088 0 RCIM_Cubic>
            {arriveTangent=1.9656, interpMode=RCIM_Cubic, leaveTangent=1.9656, time=0.2817, value=0.2693}
            {arriveTangent=0.096, interpMode=RCIM_Cubic, leaveTangent=0.096, time=0.5, value=1}
            {arriveTangent=-1.9915, interpMode=RCIM_Cubic, leaveTangent=-1.9915, time=0.6701, value=0.3066}
            <at 1.0011 0.002 RCIM_Cubic>
           ]}}}
        Scale Curve=&RandomRangeFloat(Minimum=0.07, Maximum=0.2)
       )
       Maximum=&FloatFromCurve(
        FloatCurve=DataInterface() {propertyValues {Curve {keys [
            <at -0.0088 0 RCIM_Cubic>
            {arriveTangent=2.8344, interpMode=RCIM_Cubic, leaveTangent=2.8344, time=0.1601, value=0.2359}
            {arriveTangent=0.5462, interpMode=RCIM_Cubic, leaveTangent=0.5462, time=0.3441, value=1}
            {arriveTangent=-1.5189, interpMode=RCIM_Cubic, leaveTangent=-1.5189, time=0.6254, value=0.49}
            <at 1.0011 0.002 RCIM_Cubic>
           ]}}}
        Scale Curve=&RandomRangeFloat(Minimum=0.08, Maximum=0.25)
       )
      )
     )
     3=&Multiply_Float(
      A=&FloatFromCurve(
       FloatCurve=DataInterface() {propertyValues {Curve {
          defaultValue=1
          keys [
           {arriveTangent=0.0447, interpMode=RCIM_Cubic, leaveTangent=0.0447, tangentMode=RCTM_User, time=0, value=1}
           {
            arriveTangent=-0.0466
            arriveTangentWeight=1.2798
            interpMode=RCIM_Cubic
            leaveTangent=-0.0466
            leaveTangentWeight=0.3994
            tangentMode=RCTM_User
            time=3.7126
            value=0.9789
           }
           <at 4.8745 0 RCIM_Cubic>
          ]
         }}}
       CurveIndex=$Emitter.Age
      )
      B=&FloatFromCurve(
       FloatCurve=DataInterface() {propertyValues {Curve {
          defaultValue=0
          keys [
           {interpMode=RCIM_Cubic, tangentMode=RCTM_User, time=0, value=0}
           {interpMode=RCIM_Cubic, tangentMode=RCTM_User, time=0.5, value=1}
           {tangentMode=RCTM_User, time=1, value=0}
          ]
         }}}
       Scale Curve=&Multiply_Float(
        A=&RandomRangeFloat(Minimum=0.05, Maximum=0.25)
        B=&RandomRangeFloat(
         Minimum=&FloatFromCurve(FloatCurve=DataInterface() {propertyValues {Curve {keys [
              <at 0 0.4 RCIM_Cubic>
              <at 0.1565 1 RCIM_Cubic>
              {arriveTangent=-0.1823, interpMode=RCIM_Cubic, leaveTangent=-0.1823, time=0.3307, value=0.4}
              {arriveTangent=0.611, interpMode=RCIM_Cubic, leaveTangent=0.611, time=0.4146, value=0.9529}
              {arriveTangent=0.1603, interpMode=RCIM_Cubic, leaveTangent=0.1603, time=0.5532, value=0.536}
              {arriveTangent=-0.3043, interpMode=RCIM_Cubic, leaveTangent=-0.3043, time=0.7082, value=1}
              <at 1 0.4 RCIM_Cubic>
             ]}}})
         Maximum=&FloatFromCurve(FloatCurve=DataInterface() {propertyValues {Curve {keys [
              <at 0.0035 0.5886 RCIM_Cubic>
              {arriveTangent=0.1396, interpMode=RCIM_Cubic, leaveTangent=0.1396, time=0.2798, value=1}
              <at 0.524 0.6612 RCIM_Cubic>
              {arriveTangent=-0.3388, interpMode=RCIM_Cubic, leaveTangent=-0.3388, time=0.7857, value=1}
              <at 1 0.5 RCIM_Cubic>
             ]}}})
        )
       )
      )
     )
     4=&Multiply_Float(
      A=&FloatFromCurve(FloatCurve=DataInterface() {propertyValues {Curve {keys [<at 4.8 1>,<at 5.1 0>]}}}, CurveIndex=$Emitter.Age)
      B=&MaskFloat(
       Mask Source=1
       Mask Parameter=$Particles.VisibilityTag
       Mask=0
       Masked Float=&FloatFromCurve(
        FloatCurve=DataInterface() {propertyValues {Curve {
           defaultValue=1
           keys [
            {interpMode=RCIM_Cubic, tangentMode=RCTM_User, time=0, value=0}
            {interpMode=RCIM_Cubic, tangentMode=RCTM_User, time=0.5, value=1}
            {tangentMode=RCTM_User, time=1, value=0}
           ]
          }}}
        Scale Curve=1.6
       )
       Passthrough Float=&FloatFromCurve(FloatCurve=DataInterface() {propertyValues {Curve {
           defaultValue=1
           keys [
            {interpMode=RCIM_Cubic, tangentMode=RCTM_User, time=0, value=0}
            {arriveTangent=1.0582, interpMode=RCIM_Cubic, leaveTangent=1.0582, time=0.1807, value=0.7046}
            {interpMode=RCIM_Cubic, tangentMode=RCTM_User, time=0.4096, value=0.4335}
            {arriveTangent=-0.7343, interpMode=RCIM_Cubic, leaveTangent=-0.7343, time=0.6593, value=1}
            {tangentMode=RCTM_User, time=1, value=0}
           ]
          }}})
      )
     )
     5=&Multiply_Float(
      A=&FloatFromCurve(FloatCurve=DataInterface() {propertyValues {Curve {keys [<at 4.8 1>,<at 5.1 0>]}}}, CurveIndex=$Emitter.Age)
      B=&MaskFloat(
       Mask Source=1
       Mask Parameter=$Particles.VisibilityTag
       Mask=0
       Masked Float=&FloatFromCurve(
        FloatCurve=DataInterface() {propertyValues {Curve {
           defaultValue=1
           keys [
            {interpMode=RCIM_Cubic, tangentMode=RCTM_User, time=0, value=0}
            {interpMode=RCIM_Cubic, tangentMode=RCTM_User, time=0.5, value=1}
            {tangentMode=RCTM_User, time=1, value=0}
           ]
          }}}
        Scale Curve=1.6
       )
       Passthrough Float=&FloatFromCurve(FloatCurve=DataInterface() {propertyValues {Curve {
           defaultValue=1
           keys [
            {interpMode=RCIM_Cubic, tangentMode=RCTM_User, time=0, value=0}
            {arriveTangent=1.0582, interpMode=RCIM_Cubic, leaveTangent=1.0582, time=0.1807, value=0.7046}
            {interpMode=RCIM_Cubic, tangentMode=RCTM_User, time=0.4096, value=0.4335}
            {arriveTangent=-0.7343, interpMode=RCIM_Cubic, leaveTangent=-0.7343, time=0.6593, value=1}
            {tangentMode=RCTM_User, time=1, value=0}
           ]
          }}})
      )
     )
     6=&Multiply_Float(
      A=&FloatFromCurve(
       FloatCurve=DataInterface() {propertyValues {Curve {keys [
           <at 0 0 RCIM_Cubic>
           {arriveTangent=1.0847, interpMode=RCIM_Cubic, leaveTangent=1.0847, time=0.4732, value=0.2616}
           {arriveTangentWeight=0.1523, interpMode=RCIM_Cubic, leaveTangentWeight=1.1627, tangentMode=RCTM_User, time=0.9219, value=1}
           {arriveTangentWeight=1.1562, interpMode=RCIM_Cubic, leaveTangentWeight=0.0974, tangentMode=RCTM_User, time=3.2261, value=1}
           {arriveTangent=-0.5597, interpMode=RCIM_Cubic, leaveTangent=-0.5597, time=4.0816, value=0.3049}
           <at 5.0129 0 RCIM_Cubic>
          ]}}}
       CurveIndex=$Emitter.Age
       Scale Curve=&MakeCustomFloatFromBool(
        Bool=&SetBoolByFloatComparison(
         A=&FN_PickFloatDependOnPlatform(
          TargetPlatform=DataInterface() {propertyValues {Platforms {
             qualityLevelMask=-29
             deviceProfileStates [
              {profileName=Linux, qualityLevelMask=1, setQualityLevelMask=3}
              {profileName=Mac, qualityLevelMask=1, setQualityLevelMask=3}
              {profileName=Mobile, qualityLevelMask=4294967295, setQualityLevelMask=4294967295}
              {profileName=Windows, qualityLevelMask=1, setQualityLevelMask=3}
              {profileName=Switch, qualityLevelMask=4294967295, setQualityLevelMask=4294967295}
             ]
            }}}
          TargetPlatformActive=1
          TargetPlatformNoActive=0
         )
         B=0
        )
        True Float=&RandomRangeFloat(Minimum=0.1, Maximum=0.2)
        False Float=&RandomRangeFloat(Minimum=0.25, Maximum=0.6)
       )
      )
      B=&RandomRangeFloat(
       Minimum=&FloatFromCurve(FloatCurve=DataInterface() {propertyValues {Curve {
           defaultValue=1.0113
           keys [
            {interpMode=RCIM_Cubic, tangentMode=RCTM_User, time=0, value=0}
            {interpMode=RCIM_Cubic, tangentMode=RCTM_User, time=0.3128, value=1}
            {tangentMode=RCTM_User, time=1, value=0}
           ]
          }}})
       Maximum=&FloatFromCurve(FloatCurve=DataInterface() {propertyValues {Curve {
           defaultValue=1.0113
           keys [
            {interpMode=RCIM_Cubic, tangentMode=RCTM_User, time=0, value=0}
            {interpMode=RCIM_Cubic, tangentMode=RCTM_User, time=0.6214, value=1}
            {tangentMode=RCTM_User, time=1, value=0}
           ]
          }}})
      )
     )
     7=1
     Index=&SpawnGroup()
     Num=::NewEnumerator6
    )
    Scale RGB=&MergeHelper_VectorSwitch(
     0=<1 1 1>
     1=<1 1 1>
     2=<1 1 1>
     3=&MakeVectorFromLinearColorRGB(Color=&MakeCustomLinearColorFromBool(
       Bool=&SetBoolByFloatComparison(A=0.65, B=&RandomRangeFloat(Minimum=0, Maximum=1))
       True Color=&ColorFromCurve(ColorCurve=DataInterface() {propertyValues {
          AlphaCurve {keys []}
          BlueCurve {keys [<at 0.0037 0>,<at 0.1016 0.1585>,<at 0.1773 0.1451>,<at 0.221 0.7169>,<at 0.254 1.8649>,<at 0.287 0.7358>,<at 0.3382 0.1385>,<at 0.465 0.1493>,<at 0.6461 0.3716>,<at 0.8112 0.2076>,<at 1 0>]}
          GreenCurve {keys [<at 0.0037 0.6193>,<at 0.1016 0.3508>,<at 0.1773 0.4652>,<at 0.221 1.8175>,<at 0.254 4.6742>,<at 0.287 1.8859>,<at 0.3382 0.521>,<at 0.465 0.4295>,<at 0.6461 0.8186>,<at 0.8112 0.3712>,<at 1 0.1806>]}
          RedCurve {keys [<at 0.0037 1>,<at 0.1016 1>,<at 0.1773 1>,<at 0.221 2.5>,<at 0.254 6>,<at 0.287 2.5>,<at 0.3382 1>,<at 0.465 1>,<at 0.6461 1>,<at 0.8112 1>,<at 1 1>]}
         }})
       False Color=&RandomRangeLinearColor(
        Minimum=&ColorFromCurve(ColorCurve=DataInterface() {propertyValues {
           AlphaCurve {keys []}
           BlueCurve {keys [
             <at 0.1972 1>
             <at 0.3393 0.5711>
             <at 0.4848 1>
             {arriveTangent=-0.2766, interpMode=RCIM_Cubic, leaveTangent=-0.2766, time=0.72, value=1}
             <at 1 0.3883 RCIM_Cubic>
            ]}
           GreenCurve {keys [
             <at 0.1972 1>
             <at 0.3393 0.6189>
             <at 0.4848 1>
             {arriveTangent=-0.4155, interpMode=RCIM_Cubic, leaveTangent=-0.4155, time=0.72, value=1}
             <at 1 0.3444 RCIM_Cubic>
            ]}
           RedCurve {keys [<at 0.1972 1>,<at 0.3393 1>,<at 0.4848 1>,<at 0.72 1 RCIM_Cubic>,<at 1 1 RCIM_Cubic>]}
          }})
        Maximum=&ColorFromCurve(ColorCurve=DataInterface() {propertyValues {
           AlphaCurve {keys []}
           BlueCurve {keys [
             <at 0.0587 0.5972>
             <at 0.1949 1>
             {arriveTangent=-0.3623, interpMode=RCIM_Cubic, leaveTangent=-0.3623, time=0.5163, value=1}
             {arriveTangent=-1.34, interpMode=RCIM_Cubic, leaveTangent=-1.34, time=0.8178, value=0.3221}
             <at 0.9809 0.3775>
            ]}
           GreenCurve {keys [
             <at 0.0587 0.622>
             <at 0.1949 1>
             {arriveTangent=-0.21, interpMode=RCIM_Cubic, leaveTangent=-0.21, time=0.5163, value=1}
             {arriveTangent=-1.2865, interpMode=RCIM_Cubic, leaveTangent=-1.2865, time=0.8178, value=0.4626}
             <at 0.9809 0.4023>
            ]}
           RedCurve {keys [<at 0.0587 1>,<at 0.1949 1>,<at 0.5163 1 RCIM_Cubic>,<at 0.8178 1 RCIM_Cubic>,<at 0.9809 1>]}
          }})
       )
      ))
     4=&Lerp_Vector(
      A=&MakeVectorFromLinearColorRGB(Color=&MakeCustomLinearColorFromBool(
        Bool=&SetBoolByFloatComparison(A=0.55, B=&RandomRangeFloat(Minimum=0, Maximum=1))
        True Color=&ColorFromCurve(ColorCurve=DataInterface() {propertyValues {
           AlphaCurve {keys []}
           BlueCurve {keys [
             <at 0 0 RCIM_Cubic>
             {arriveTangent=0.3312, interpMode=RCIM_Cubic, leaveTangent=0.3312, time=0.2164, value=0.1685}
             <at 0.3292 0.109>
             <at 0.3597 0.128>
             <at 0.4157 2.6508>
             <at 0.4751 0.1385>
             <at 0.535 0.1585>
             <at 0.6895 0.3359>
             <at 0.9946 0.1182>
            ]}
           GreenCurve {keys [
             <at 0 0.1315 RCIM_Cubic>
             {arriveTangent=1.9451, interpMode=RCIM_Cubic, leaveTangent=1.9451, time=0.2164, value=0.6317}
             <at 0.3292 0.7719>
             <at 0.3597 0.6106>
             <at 0.4157 11.4361>
             <at 0.4751 0.521>
             <at 0.535 0.3508>
             <at 0.6895 0.4843>
             <at 0.9946 0.0895>
            ]}
           RedCurve {keys [<at 0 1 RCIM_Cubic>,<at 0.2164 1 RCIM_Cubic>,<at 0.3292 1>,<at 0.3597 1>,<at 0.4157 20>,<at 0.4751 1>,<at 0.535 1>,<at 0.6895 1>,<at 0.9946 1>]}
          }})
        False Color=&ColorFromCurve(ColorCurve=DataInterface() {propertyValues {
           AlphaCurve {keys []}
           BlueCurve {keys [<at 0.0037 0>,<at 0.1822 0.1585>,<at 0.2933 0.1493>,<at 0.3717 3.1081>,<at 0.4446 0.1451>,<at 0.5143 0.1385>,<at 0.7211 0.2464>,<at 0.8387 0.2076>,<at 1 0>]}
           GreenCurve {keys [<at 0.0037 0.6193>,<at 0.1822 0.3508>,<at 0.2933 0.4295>,<at 0.3717 7.7904>,<at 0.4446 0.4652>,<at 0.5143 0.521>,<at 0.7211 0.8521>,<at 0.8387 0.3712>,<at 1 0.1806>]}
           RedCurve {keys [<at 0.0037 1>,<at 0.1822 1>,<at 0.2933 1>,<at 0.3717 10>,<at 0.4446 1>,<at 0.5143 1>,<at 0.7211 1>,<at 0.8387 1>,<at 1 1>]}
          }})
       ))
      B=&MakeVectorFromLinearColorRGB(Color=&MakeCustomLinearColorFromBool(
        Bool=&SetBoolByFloatComparison(A=0.55, B=&RandomRangeFloat(Minimum=0, Maximum=1))
        True Color=&ColorFromCurve(ColorCurve=DataInterface() {propertyValues {
           AlphaCurve {keys []}
           BlueCurve {keys [<at 0.0091 0.1182>,<at 0.0733 0.3359>,<at 0.1365 0.0985>,<at 0.1779 1.0778>,<at 0.2236 0>,<at 0.2454 0>,<at 0.3292 0.109>,<at 0.3597 0.128>,<at 0.4157 2.6508>,<at 0.4751 0.1385>,<at 0.535 0.1585>,<at 0.7309 0.067>,<at 0.7701 2.2271>,<at 0.806 0.2527>,<at 0.8812 0.2271>,<at 0.9585 0 RCIM_Cubic>]}
           GreenCurve {keys [<at 0.0091 0.0895>,<at 0.0733 0.4843>,<at 0.1365 0.4073>,<at 0.1779 4.1875>,<at 0.2236 0.4223>,<at 0.2454 0.3057>,<at 0.3292 0.7719>,<at 0.3597 0.6106>,<at 0.4157 11.4361>,<at 0.4751 0.521>,<at 0.535 0.3508>,<at 0.7309 0.3065>,<at 0.7701 3.0403>,<at 0.806 0.3335>,<at 0.8812 0.287>,<at 0.9585 0.1315 RCIM_Cubic>]}
           RedCurve {keys [<at 0.0091 1>,<at 0.0733 1>,<at 0.1365 1>,<at 0.1779 8>,<at 0.2236 1>,<at 0.2454 1>,<at 0.3292 1>,<at 0.3597 1>,<at 0.4157 20>,<at 0.4751 1>,<at 0.535 1>,<at 0.7309 1>,<at 0.7701 8>,<at 0.806 1>,<at 0.8812 1>,<at 0.9585 1 RCIM_Cubic>]}
          }})
        False Color=&ColorFromCurve(ColorCurve=DataInterface() {propertyValues {
           AlphaCurve {keys []}
           BlueCurve {keys [<at 0.0037 0.3368>,<at 0.1822 0.1585>,<at 0.2933 0.1493>,<at 0.3717 3.1081>,<at 0.4446 0.1451>,<at 0.5143 0.1385>,<at 0.7211 0.3716>,<at 0.8387 0.2076>,<at 1 0>]}
           GreenCurve {keys [<at 0.0037 0.2391>,<at 0.1822 0.3508>,<at 0.2933 0.4295>,<at 0.3717 7.7904>,<at 0.4446 0.4652>,<at 0.5143 0.521>,<at 0.7211 0.8186>,<at 0.8387 0.3712>,<at 1 0.1806>]}
           RedCurve {keys [<at 0.0037 1>,<at 0.1822 1>,<at 0.2933 1>,<at 0.3717 10>,<at 0.4446 1>,<at 0.5143 1>,<at 0.7211 1>,<at 0.8387 1>,<at 1 1>]}
          }})
       ))
      Alpha=&UniformAOrBFloat(A=0, Evaluation Type=::NewEnumerator0, Recalculate Random Each Loop=true, Override Seed=false)
     )
     5=&Lerp_Vector(
      A=&MakeVectorFromLinearColorRGB(Color=&MakeCustomLinearColorFromBool(
        Bool=&SetBoolByFloatComparison(A=0.55, B=&RandomRangeFloat(Minimum=0, Maximum=1))
        True Color=&ColorFromCurve(ColorCurve=DataInterface() {propertyValues {
           AlphaCurve {keys []}
           BlueCurve {keys [
             <at 0 0 RCIM_Cubic>
             {arriveTangent=0.3312, interpMode=RCIM_Cubic, leaveTangent=0.3312, time=0.2164, value=0.1685}
             <at 0.3292 0.109>
             <at 0.3597 0.128>
             <at 0.4157 2.6508>
             <at 0.4751 0.1385>
             <at 0.535 0.1585>
             <at 0.6895 0.3359>
             <at 0.9946 0.1182>
            ]}
           GreenCurve {keys [
             <at 0 0.1315 RCIM_Cubic>
             {arriveTangent=1.9451, interpMode=RCIM_Cubic, leaveTangent=1.9451, time=0.2164, value=0.6317}
             <at 0.3292 0.7719>
             <at 0.3597 0.6106>
             <at 0.4157 11.4361>
             <at 0.4751 0.521>
             <at 0.535 0.3508>
             <at 0.6895 0.4843>
             <at 0.9946 0.0895>
            ]}
           RedCurve {keys [<at 0 1 RCIM_Cubic>,<at 0.2164 1 RCIM_Cubic>,<at 0.3292 1>,<at 0.3597 1>,<at 0.4157 20>,<at 0.4751 1>,<at 0.535 1>,<at 0.6895 1>,<at 0.9946 1>]}
          }})
        False Color=&ColorFromCurve(ColorCurve=DataInterface() {propertyValues {
           AlphaCurve {keys []}
           BlueCurve {keys [<at 0.0037 0>,<at 0.1822 0.1585>,<at 0.2933 0.1493>,<at 0.3717 3.1081>,<at 0.4446 0.1451>,<at 0.5143 0.1385>,<at 0.7211 0.2464>,<at 0.8387 0.2076>,<at 1 0>]}
           GreenCurve {keys [<at 0.0037 0.6193>,<at 0.1822 0.3508>,<at 0.2933 0.4295>,<at 0.3717 7.7904>,<at 0.4446 0.4652>,<at 0.5143 0.521>,<at 0.7211 0.8521>,<at 0.8387 0.3712>,<at 1 0.1806>]}
           RedCurve {keys [<at 0.0037 1>,<at 0.1822 1>,<at 0.2933 1>,<at 0.3717 10>,<at 0.4446 1>,<at 0.5143 1>,<at 0.7211 1>,<at 0.8387 1>,<at 1 1>]}
          }})
       ))
      B=&MakeVectorFromLinearColorRGB(Color=&MakeCustomLinearColorFromBool(
        Bool=&SetBoolByFloatComparison(
         A=0.55
         B=&UniformAOrBFloat(A=0, Evaluation Type=::NewEnumerator0, Recalculate Random Each Loop=true, Override Seed=false)
        )
        True Color=&ColorFromCurve(ColorCurve=DataInterface() {propertyValues {
           AlphaCurve {keys []}
           BlueCurve {keys [<at 0.0091 0.1182>,<at 0.0733 0.3359>,<at 0.1365 0.0985>,<at 0.1779 1.0778>,<at 0.2236 0>,<at 0.2454 0>,<at 0.3292 0.109>,<at 0.3597 0.128>,<at 0.4157 2.6508>,<at 0.4751 0.1385>,<at 0.535 0.1585>,<at 0.7309 0.067>,<at 0.7701 2.2271>,<at 0.806 0.2527>,<at 0.8812 0.2271>,<at 0.9585 0 RCIM_Cubic>]}
           GreenCurve {keys [<at 0.0091 0.0895>,<at 0.0733 0.4843>,<at 0.1365 0.4073>,<at 0.1779 4.1875>,<at 0.2236 0.4223>,<at 0.2454 0.3057>,<at 0.3292 0.7719>,<at 0.3597 0.6106>,<at 0.4157 11.4361>,<at 0.4751 0.521>,<at 0.535 0.3508>,<at 0.7309 0.3065>,<at 0.7701 3.0403>,<at 0.806 0.3335>,<at 0.8812 0.287>,<at 0.9585 0.1315 RCIM_Cubic>]}
           RedCurve {keys [<at 0.0091 1>,<at 0.0733 1>,<at 0.1365 1>,<at 0.1779 8>,<at 0.2236 1>,<at 0.2454 1>,<at 0.3292 1>,<at 0.3597 1>,<at 0.4157 20>,<at 0.4751 1>,<at 0.535 1>,<at 0.7309 1>,<at 0.7701 8>,<at 0.806 1>,<at 0.8812 1>,<at 0.9585 1 RCIM_Cubic>]}
          }})
        False Color=&ColorFromCurve(ColorCurve=DataInterface() {propertyValues {
           AlphaCurve {keys []}
           BlueCurve {keys [<at 0.0037 0.3368>,<at 0.1822 0.1585>,<at 0.2933 0.1493>,<at 0.3717 3.1081>,<at 0.4446 0.1451>,<at 0.5143 0.1385>,<at 0.7211 0.3716>,<at 0.8387 0.2076>,<at 1 0>]}
           GreenCurve {keys [<at 0.0037 0.2391>,<at 0.1822 0.3508>,<at 0.2933 0.4295>,<at 0.3717 7.7904>,<at 0.4446 0.4652>,<at 0.5143 0.521>,<at 0.7211 0.8186>,<at 0.8387 0.3712>,<at 1 0.1806>]}
           RedCurve {keys [<at 0.0037 1>,<at 0.1822 1>,<at 0.2933 1>,<at 0.3717 10>,<at 0.4446 1>,<at 0.5143 1>,<at 0.7211 1>,<at 0.8387 1>,<at 1 1>]}
          }})
       ))
      Alpha=&RandomRangeFloat(Minimum=0, Maximum=1)
     )
     6=&MakeVectorFromLinearColorRGB(Color=&ColorFromCurve(ColorCurve=DataInterface() {propertyValues {
         AlphaCurve {keys []}
         BlueCurve {keys [<at 0 0.4915 RCIM_Cubic>,<at 0.1464 0.4534>,<at 0.3102 0.9913>,<at 0.4699 1>,<at 0.6389 0.954>,<at 0.8068 0.1552>,<at 1 0.4809 RCIM_Cubic>]}
         GreenCurve {keys [<at 0 0.5394 RCIM_Cubic>,<at 0.1464 0.6105>,<at 0.3102 0.9921>,<at 0.4699 1>,<at 0.6389 0.9569>,<at 0.8068 0.3737>,<at 1 0.5134 RCIM_Cubic>]}
         RedCurve {keys [<at 0 1 RCIM_Cubic>,<at 0.1464 1>,<at 0.3102 1>,<at 0.4699 1>,<at 0.6389 1>,<at 0.8068 1>,<at 1 1 RCIM_Cubic>]}
        }}))
     7=<1 1 1>
     Index=&SpawnGroup()
     Num=::NewEnumerator6
    )
    ScaleA=true
    ScaleRGB=true
   )
   &SubUVAnimation(
    End Frame Range Override=64
    Output Alternate Renderer Binding=true
    "Sprite Renderer "=DataInterface() {propertyValues {SpriteRenderer=&self:ConesMistCondensed_0.NiagaraSpriteRendererProperties_1}}
    SubUV Animation Mode:ENiagara_SubUVLookupModeV2=::NewEnumerator1
    SubUV Curve=DataInterface() {propertyValues {Curve {keys [<at 0 0 RCIM_Cubic>,<at 1 0.4 RCIM_Cubic>]}}}
    UseEndFrame=true
    UseStartFrame=false
   )
   &GravityForce(Gravity=&MergeHelper_VectorSwitch(
     0=<0 0 0>
     1=<0 0 0>
     2=&RandomRangeVector(Minimum=<0 0 -100>, Maximum=<0 0 -250>)
     3=&RandomRangeVector(Minimum=<0 0 -60>, Maximum=<0 0 -120>)
     4=&RandomRangeVector(Minimum=<0 0 -60>, Maximum=<0 0 -120>)
     5=&RandomRangeVector(Minimum=<0 0 -60>, Maximum=<0 0 -120>)
     6=&RandomRangeVector(Minimum=<0 0 -75>, Maximum=<0 0 -250>)
     7=<0 0 0>
     Index=&SpawnGroup()
     Num=::NewEnumerator6
    ))
   &VortexForce(
    Origin Pull Amount:NiagaraFloat=&MergeHelper_FloatSwitch(0=0, 1=0, 2=0, 3=20, 4=10, 5=10, 6=0, 7=0, Index=&SpawnGroup(), Num=::NewEnumerator6)
    UseOriginOffset=false
    Vortex Force Amount=&MergeHelper_FloatSwitch(
     0=0
     1=0
     2=&RandomRangeFloat(Minimum=80, Maximum=200)
     3=&RandomRangeFloat(Minimum=200, Maximum=400)
     4=&RandomRangeFloat(Minimum=200, Maximum=400)
     5=&RandomRangeFloat(Minimum=200, Maximum=400)
     6=0
     7=0
     Index=&SpawnGroup()
     Num=::NewEnumerator6
    )
   )
   &CurlNoiseForce(
    Bias Noise=false
    Noise Frequency=&MergeHelper_FloatSwitch(
     0=0
     1=0
     2=30
     3=&RandomRangeFloat(Minimum=30, Maximum=70)
     4=&RandomRangeFloat(Minimum=30, Maximum=70)
     5=&RandomRangeFloat(Minimum=30, Maximum=70)
     6=&RandomRangeFloat(Minimum=30, Maximum=60)
     7=0
     Index=&SpawnGroup()
     Num=::NewEnumerator6
    )
    Noise Strength=&MergeHelper_FloatSwitch(
     0=0
     1=0
     2=100
     3=&RandomRangeFloat(Minimum=20, Maximum=50)
     4=&RandomRangeFloat(Minimum=20, Maximum=50)
     5=&RandomRangeFloat(Minimum=20, Maximum=50)
     6=&RandomRangeFloat(Minimum=50, Maximum=100)
     7=0
     Index=&SpawnGroup()
     Num=::NewEnumerator6
    )
    Pan Noise=false
   )
   &Drag(
    Drag=&MergeHelper_FloatSwitch(
     0=0
     1=0
     2=&RandomRangeFloat(Minimum=3, Maximum=2)
     3=&RandomRangeFloat(Minimum=1.5, Maximum=2.5)
     4=&RandomRangeFloat(Minimum=1.5, Maximum=2.5)
     5=&RandomRangeFloat(Minimum=1.5, Maximum=2.5)
     6=&RandomRangeFloat(Minimum=3, Maximum=6)
     7=0
     Index=&SpawnGroup()
     Num=::NewEnumerator6
    )
    Rotational Drag:NiagaraFloat=&MaskFloat(
     Mask Source=0
     Mask Parameter=?
     Mask=6
     Masked Float=&RandomRangeFloat(Minimum=0.5, Maximum=2)
     Passthrough Float=1
    )
    Use Linear Drag=true
    Use Rotational Drag=true
   )
   /SetVariables(
    Particles.Bottom Cone Size:Vector3f=<1.3 1.3 1.25>
    Particles.Cones Dynamic Param:Vector4f=&MakeVector4(X=0.2, Y=0, Z=0, W=0)
    Particles.Inner Cone Size:Vector3f=<1 1 1.36>
    Particles.Outer Cone Size:Vector3f=<2.5 2.5 1.5>
    Particles.Top Cone Size:Vector3f=<2.2 2.2 1.2>
   ) {isSetParameters=true}
   /SetVariables(
    Particles.Inner Cone Color:LinearColor=&MakeLinearColorFromVectorAndFloat(
     "Vector (RGB)"=&MakeVectorFromLinearColorRGB(Color=<rgba 1.25 0.8671 0 0.5861>)
     "Float (Alpha)"=&FloatFromCurve(
      FloatCurve=DataInterface() {propertyValues {Curve {keys [
          <at 0 0 RCIM_Cubic>
          {arriveTangent=1.0847, interpMode=RCIM_Cubic, leaveTangent=1.0847, time=0.4732, value=0.2616}
          {arriveTangentWeight=0.1523, interpMode=RCIM_Cubic, leaveTangentWeight=1.1627, tangentMode=RCTM_User, time=0.9219, value=1}
          {arriveTangentWeight=0.7975, interpMode=RCIM_Cubic, leaveTangentWeight=0.2232, tangentMode=RCTM_User, time=2.4338, value=1}
          {
           arriveTangent=-0.7604
           arriveTangentWeight=1.1562
           interpMode=RCIM_Cubic
           leaveTangent=-0.7604
           leaveTangentWeight=0.0974
           tangentMode=RCTM_User
           time=3.1486
           value=0.4471
          }
          <at 3.9938 -0.0152 RCIM_Cubic>
         ]}}}
      CurveIndex=$Emitter.Age
      Scale Curve=0.6
     )
    )
    Particles.Outer Cone Color:LinearColor=&MakeLinearColorFromVectorAndFloat(
     "Vector (RGB)"=&MakeVectorFromLinearColorRGB(Color=<rgba 1.1 0.6289 0 1.15>)
     "Float (Alpha)"=&FloatFromCurve(
      FloatCurve=DataInterface() {propertyValues {Curve {keys [
          <at 0 0 RCIM_Cubic>
          {arriveTangent=1.0847, interpMode=RCIM_Cubic, leaveTangent=1.0847, time=0.4732, value=0.2616}
          {arriveTangentWeight=0.1523, interpMode=RCIM_Cubic, leaveTangentWeight=1.1627, tangentMode=RCTM_User, time=0.9219, value=1}
          {arriveTangentWeight=0.7975, interpMode=RCIM_Cubic, leaveTangentWeight=0.2232, tangentMode=RCTM_User, time=3.5251, value=1}
          {
           arriveTangent=-0.9526
           arriveTangentWeight=1.1562
           interpMode=RCIM_Cubic
           leaveTangent=-0.9526
           leaveTangentWeight=0.0974
           tangentMode=RCTM_User
           time=4.2349
           value=0.5571
          }
          <at 4.8452 0 RCIM_Cubic>
         ]}}}
      CurveIndex=$Emitter.Age
      Scale Curve=&MakeCustomFloatFromBool(Bool=&Boolean_OrOperation(Bool1=$User.bIsFrontEnd, Bool2=$User.bIsFrontEndPreview), True Float=1, False Float=1)
     )
    )
    Particles.Top Color:LinearColor=&MakeLinearColorFromVectorAndFloat(
     "Vector (RGB)"=&MakeVectorFromLinearColorRGB(Color=<rgba 2 1.0034 0.0598 0.97>)
     "Float (Alpha)"=&FloatFromCurve(
      FloatCurve=DataInterface() {propertyValues {Curve {keys [
          <at 0 0 RCIM_Cubic>
          {arriveTangent=1.0847, interpMode=RCIM_Cubic, leaveTangent=1.0847, time=0.4732, value=0.2616}
          {arriveTangentWeight=0.1523, interpMode=RCIM_Cubic, leaveTangentWeight=1.1627, tangentMode=RCTM_User, time=0.9219, value=1}
          {arriveTangentWeight=0.7975, interpMode=RCIM_Cubic, leaveTangentWeight=0.2232, tangentMode=RCTM_User, time=2.0562, value=1}
          {
           arriveTangent=-0.7604
           arriveTangentWeight=1.1562
           interpMode=RCIM_Cubic
           leaveTangent=-0.7604
           leaveTangentWeight=0.0974
           tangentMode=RCTM_User
           time=2.8736
           value=0.4282
          }
          <at 3.7701 0 RCIM_Cubic>
         ]}}}
      CurveIndex=$Emitter.Age
      Scale Curve=&Multiply_Float(
       A=0.35
       B=&Multiply_Float(
        A=&FNTimeOfDayScalar(Night Scale=0.8, Morning Scale=1, Evening Scale=1)
        B=&FN_PickFloatDependOnPlatform(TargetPlatformActive=1, TargetPlatformNoActive=0.6)
       )
      )
     )
    )
   ) {isSetParameters=true}
   /SetVariables(
    Particles.Falling Mist Dynamic Param:Vector4f=&MakeVector4(X=0, Y=1, Z=&RandomRangeFloat(Minimum=0.05, Maximum=0.2), W=1)
    Particles.God Rays Dynamic Param:Vector4f=&MakeVector4(X=&RandomRangeFloat(Minimum=0.05, Maximum=-0.05), Y=0, Z=0, W=0)
    Particles.Top Cone Dynamic Param:Vector4f=&MakeVector4(
     X=&MakeCustomFloatFromBool(Bool=$User.bIsFrontEndPreview, True Float=0.2, False Float=0.5)
     Y=1
     Z=0
     W=1
    )
   ) {isSetParameters=true}
   /SetVariables(Particles.Rainbow Flare Color:LinearColor=&MultiplyLinearColor(
     Linear Color=<rgba 1 0.8714 0.4965 1>
     Scale Factor=&MakeVector4_FromLinearColor(LinearColor=&ColorFromCurve(ColorCurve=DataInterface() {propertyValues {
         AlphaCurve {keys [<at 0.0114 0>,<at 0.0966 0.2>,<at 0.3226 0.75>,<at 0.5815 0.6>,<at 0.9914 0>]}
         BlueCurve {keys [<at 0 0.0428>,<at 0.2101 0.0808>,<at 0.3981 0.1569>,<at 0.583 0.1327>,<at 0.7294 0.1135>,<at 0.9929 0.1618>]}
         GreenCurve {keys [<at 0 0.1828>,<at 0.2101 0.5211>,<at 0.3981 0.8081>,<at 0.583 0.5116>,<at 0.7294 0.2769>,<at 0.9929 0.179>]}
         RedCurve {keys [<at 0 1>,<at 0.2101 1>,<at 0.3981 1>,<at 0.583 1>,<at 0.7294 1>,<at 0.9929 1>]}
        }}))
    )) {isSetParameters=true}
   &SolveForcesAndVelocity(Clamp Velocity=false, Limit Acceleration=false)
   &ScaleSpriteSizeBySpeed(
    Max Scale Factor=&MergeHelper_Vector2Switch(
     0=<1 1>
     1=<1 1>
     2=<1 1>
     3=<1 1>
     4=<1 1.3>
     5=<1 1.3>
     6=<1 1>
     7=<1 1>
     Index=&SpawnGroup()
     Num=::NewEnumerator6
    )
    Min Scale Factor=&MergeHelper_Vector2Switch(
     0=<1 1>
     1=<1 1>
     2=<1 1>
     3=<1 1>
     4=<0.5 1>
     5=<0.5 1>
     6=<1 1>
     7=<1 1>
     Index=&SpawnGroup()
     Num=::NewEnumerator6
    )
   )
   /SetVariables(
    Particles.Front Light Color:LinearColor=&ColorFromCurve(ColorCurve=DataInterface() {propertyValues {
       AlphaCurve {keys [<at 0.0326 0>,<at 0.3153 1>,<at 0.5549 1>,<at 0.8188 0.25>,<at 0.9722 0>]}
       BlueCurve {keys [<at 0.1457 0.4812>,<at 0.2884 0.9645>,<at 0.598 0.4812>]}
       GreenCurve {keys [<at 0.1457 0.9618>,<at 0.2884 2.1408>,<at 0.598 0.9618>]}
       RedCurve {keys [<at 0.1457 1.5>,<at 0.2884 3>,<at 0.598 1.5>]}
      }})
    Particles.Top Light Color:LinearColor=&ColorFromCurve(ColorCurve=DataInterface() {propertyValues {
       AlphaCurve {keys [<at 0.0326 0>,<at 0.3153 1>,<at 0.5549 1>,<at 0.8188 0.25>,<at 0.9722 0>]}
       BlueCurve {keys [<at 0.1457 0.6416>,<at 0.2884 1.286>,<at 0.598 0.6416>]}
       GreenCurve {keys [<at 0.1457 1.2823>,<at 0.2884 2.8544>,<at 0.598 1.2823>]}
       RedCurve {keys [<at 0.1457 2>,<at 0.2884 4>,<at 0.598 2>]}
      }})
   ) {isSetParameters=true}
   &Light_Attributes(
    Diffuse Scale=1
    Falloff Exponent=1
    Light Position Offset:Vector3f=<0 0 150>
    Position=1
    Radius=1
    Specular Scale=1
    UsePositionOffset=true
    Volumetric Scattering=1
   )
  ]
  rendererClasses [
   &NiagaraMeshRendererProperties
   &NiagaraSpriteRendererProperties
   &NiagaraLightRendererProperties
  ]
  renderers [
   {rendererClass=&NiagaraMeshRendererProperties, rendererIndex=0}
   {rendererClass=&NiagaraMeshRendererProperties, rendererIndex=1}
   {rendererClass=&NiagaraMeshRendererProperties, rendererIndex=2}
   {rendererClass=&NiagaraMeshRendererProperties, rendererIndex=3}
   {rendererClass=&NiagaraMeshRendererProperties, rendererIndex=4}
   {rendererClass=&NiagaraSpriteRendererProperties, rendererIndex=5}
   {rendererClass=&NiagaraSpriteRendererProperties, rendererIndex=6}
   {rendererClass=&NiagaraSpriteRendererProperties, rendererIndex=7}
   {rendererClass=&NiagaraSpriteRendererProperties, rendererIndex=8}
   {rendererClass=&NiagaraSpriteRendererProperties, rendererIndex=9}
   {rendererClass=&NiagaraSpriteRendererProperties, rendererIndex=10}
   {rendererClass=&NiagaraLightRendererProperties, rendererIndex=11}
   {rendererClass=&NiagaraLightRendererProperties, rendererIndex=12}
  ]
 }
 Ground_Glow {rendererClasses [], renderers []}
 Joint_Glows {
  emitterUpdateScript [
   &EmitterState(Inactive Response=::NewEnumerator1, Life Cycle Mode=::NewEnumerator1, Loop Behavior=::NewEnumerator1, Loop Duration=5.1, UseLoopDelay=false)
   &SpawnBurst_Instantaneous(Spawn Count=14, Use Spawn Probability=false, UseLoopCountLimit=false)
  ]
  particleSpawnScript [
   &PartitionParticles(Number Of Partitions=14)
   &InitializeParticle(Color Mode=::NewEnumerator0, Lifetime=5.1, UsePositionOffset=false, Write Material Random=true)
   &CameraOffset(Camera Offset Amount=10)
   /SetVariables(Particles.GlowScale:Vector2f=&Multiply_Vector2D(
     A=&SelectVector2DFromArray(
      Direct Array Index=$Particles.PartitionParticles.Partition
      Vector 2D Selection Array=DataInterface() {propertyValues {FloatData [<150 150>, <150 150>, <80 80>, <80 80>, <100 100>, <100 100>, <130 130>, <150 150>, <150 150>, <90 90>, <90 90>, <50 50>, <50 50>, <180 180>]}}
     )
     B=<1 1>
    )) {isSetParameters=true}
   /SetVariables(Particles.GlowCameraOffset=-10) {isSetParameters=true}
   /SetVariables(Particles.GlowColor=&Multiply_LinearColorByFloat(
     Linear Color=<rgba 2 1.4246 0.5021 0.7>
     Float=&FNTimeOfDayScalar(Night Scale=0.5, Morning Scale=0.78, Evening Scale=0.78)
    )) {isSetParameters=true}
  ]
  particleUpdateScript [
   &ParticleState()
   &SpriteRotationRate(Rotation Rate=&RandomRangeFloat(Minimum=-20, Maximum=40))
   &SkeletalMeshLocation(
    Bone / Socket Index=&SelectIntFromArray(
     Direct Array Index=$Particles.PartitionParticles.Partition
     Int Selection Array=DataInterface() {propertyValues {IntData [0,1,2,3,4,5,6,7,8,9,10,11,12,13]}}
    )
    Bone Sampling Mode=::NewEnumerator4
    MaskSpawnGroup=false
    Orientation Sampling=::NewEnumerator1
    Sampled Position Offset=&MaskVector(Mask Source=1, Mask Parameter=$Particles.PartitionParticles.Partition, Mask=15, Masked Vector=<0 0 6>, Passthrough Vector=<0 0 0>)
    Sampled Position Offset Coordinate Space:ENiagaraSkelMeshTransforms=::NewEnumerator2
    Skeletal Mesh=$System.Skeletal Mesh
   )
   &SolveForcesAndVelocity(Clamp Velocity=false, Limit Acceleration=false)
   &DynamicMaterialParameters(
    Index 0 Param 1=5
    Index 0 Param 2=1
    Index 0 Param 3=1
    Index 0 Param 4=1
    Module.Index0Param0Write=true
    Module.Index0Param1Write=true
    Module.Index0Param2Write=true
    Module.Index0Param3Write=true
   )
   /SetVariables(Particles.Scale Glow Color:LinearColor=&MakeLinearColorFromVectorAndFloat(
     "Vector (RGB)"=&MakeVectorFromLinearColorRGB(Color=$Particles.Initial.GlowColor)
     "Float (Alpha)"=&FloatFromCurve(
      FloatCurve=DataInterface() {propertyValues {Curve {
         defaultValue=0
         keys [
          {interpMode=RCIM_Cubic, tangentMode=RCTM_User, time=0.0525, value=0}
          {arriveTangent=2.8346, interpMode=RCIM_Cubic, leaveTangent=2.8346, time=0.2023, value=0.1654}
          {arriveTangent=4.2385, interpMode=RCIM_Cubic, leaveTangent=4.2385, time=0.2943, value=0.6855}
          {interpMode=RCIM_Cubic, tangentMode=RCTM_User, time=0.3977, value=0.9938}
          {arriveTangent=-2.9923, interpMode=RCIM_Cubic, leaveTangent=-2.9923, time=0.5318, value=0.317}
          {arriveTangent=-0.677, interpMode=RCIM_Cubic, leaveTangent=-0.677, time=0.7154, value=0.0433}
          {tangentMode=RCTM_User, time=1, value=0}
         ]
        }}}
      Scale Curve=0.6
     )
    )) {isSetParameters=true}
  ]
  rendererClasses [&NiagaraSpriteRendererProperties]
  renderers [{rendererClass=&NiagaraSpriteRendererProperties}]
 }
 Motion_Trails {
  emitterUpdateScript [
   &EmitterState(Life Cycle Mode=::NewEnumerator1, Loop Behavior=::NewEnumerator1, Loop Duration=5.1, UseLoopDelay=false)
   &SpawnRate(
    SpawnRate=&Multiply_Float(
     A=30
     B=&FloatFromCurve(
      FloatCurve=DataInterface() {propertyValues {Curve {keys [<at 2.0952 0>, <at 2.2107 1>, {arriveTangent=-0.7226, leaveTangent=-0.7226, time=3.5493, value=1.0142}, <at 3.7003 0>]}}}
      CurveIndex=$Emitter.Age
     )
    )
    Use Spawn Probability=false
   )
  ]
  particleSpawnScript [
   &InitializeParticle(
    Color=<rgba 1 0.7273 0.0929 0.7704>
    Lifetime=0.65
    Ribbon Width=15
    Ribbon Width Mode=::NewEnumerator1
    UsePositionOffset=false
    Write Material Random=true
   )
   &SkeletalMeshLocation(MaskSpawnGroup=false, Mesh Sampling Type=::NewEnumerator5, Skeletal Mesh=$System.left hand, Socket Sampling Mode=::NewEnumerator5)
  ]
  particleUpdateScript [
   &ParticleState()
   &ScaleRibbonWidth(Ribbon Width Scale=&FloatFromCurve(FloatCurve=DataInterface() {propertyValues {Curve {
        defaultValue=1
        keys [
         <at -0.0035 0 RCIM_Cubic>
         {arriveTangent=0.1693, interpMode=RCIM_Cubic, leaveTangent=0.1693, tangentMode=RCTM_User, time=0.2185, value=1.0569}
         {arriveTangent=-1.5779, interpMode=RCIM_Cubic, leaveTangent=-1.5779, tangentMode=RCTM_User, time=0.4225, value=0.9052}
         {tangentMode=RCTM_User, time=1, value=0}
        ]
       }}}))
   &ScaleColor(
    Scale Alpha=&Multiply_Float(
     A=&FloatFromCurve(FloatCurve=DataInterface() {propertyValues {Curve {
         defaultValue=1
         keys [
          {interpMode=RCIM_Cubic, tangentMode=RCTM_User, time=0, value=1}
          {arriveTangent=-1.1765, interpMode=RCIM_Cubic, leaveTangent=-1.1765, time=0.2936, value=0.4828}
          {interpMode=RCIM_Cubic, tangentMode=RCTM_User, time=0.85, value=0}
         ]
        }}})
     B=&FloatFromCurve(
      FloatCurve=DataInterface() {propertyValues {Curve {keys [<at 2.1 0>,<at 2.497 0.304>,<at 2.8008 1>,<at 4.6664 1>,<at 4.8 0>]}}}
      CurveIndex=$Emitter.Age
     )
    )
    ScaleA=true
    ScaleRGB=true
   )
   &Drag(Drag=5, Use Linear Drag=true, Use Rotational Drag=true)
   &Drag(Drag=5, Use Linear Drag=true, Use Rotational Drag=true) {module=Drag001}
   &CurlNoiseForce(Bias Noise=false, Noise Strength=50, Pan Noise=false)
   &SolveForcesAndVelocity(Clamp Velocity=false, Limit Acceleration=false)
  ]
  rendererClasses [&NiagaraRibbonRendererProperties]
  renderers [{rendererClass=&NiagaraRibbonRendererProperties}]
 }
 Res_BuildUp_Sparks002 {
  emitterUpdateScript [
   &EmitterState(Life Cycle Mode=::NewEnumerator1, Loop Behavior=::NewEnumerator1, Loop Duration=5.1, UseLoopDelay=false)
   &SpawnRate(
    SpawnRate=&Multiply_Float(
     A=200
     B=&FloatFromCurve(
      FloatCurve=DataInterface() {propertyValues {Curve {keys [
          <at 2.4 0 RCIM_Cubic>
          {
           arriveTangent=-0.4227
           arriveTangentWeight=0.1146
           interpMode=RCIM_Cubic
           leaveTangent=105.4924
           leaveTangentWeight=0.2689
           tangentMode=RCTM_Break
           time=2.5606
           value=0.5582
          }
          {arriveTangent=7.4099, interpMode=RCIM_Cubic, leaveTangent=7.4099, tangentMode=RCTM_User, time=2.6058, value=1.114}
          {
           arriveTangent=1.9112
           arriveTangentWeight=0.2874
           interpMode=RCIM_Cubic
           leaveTangent=59.7378
           leaveTangentWeight=0.0887
           tangentMode=RCTM_Break
           time=3.0589
           value=1.9575
          }
          {arriveTangent=-41.1631, interpMode=RCIM_Cubic, leaveTangent=-41.1631, tangentMode=RCTM_User, time=3.0935, value=1.6776}
          {arriveTangent=1.5633, interpMode=RCIM_Cubic, leaveTangent=1.5633, tangentMode=RCTM_User, time=3.1655, value=0.023}
         ]}}}
      CurveIndex=$Emitter.Age
      Scale Curve=&FloatFromCurve(
       FloatCurve=DataInterface() {propertyValues {Curve {defaultValue=0, keys [{tangentMode=RCTM_User, time=0, value=0}, {tangentMode=RCTM_User, time=1, value=1}]}}}
       CurveIndex=$Emitter.NormalizedLoopAge
      )
     )
    )
    Use Spawn Probability=false
   )
  ]
  particleSpawnScript [&InitializeParticle(
    Color Maximum=<rgba 12 8.0966 0 1>
    Color Minimum=<rgba 7 5.3698 1.9175 1>
    Color Mode=::NewEnumerator2
    Lifetime Max=1.2
    Lifetime Min=0.6
    Lifetime Mode=::NewEnumerator1
    Mass Mode=::NewEnumerator3
    Ribbon Width=8
    Ribbon Width Mode=::NewEnumerator1
    Sprite Rotation Mode=::NewEnumerator2
    Sprite Size Mode=::NewEnumerator1
    Uniform Sprite Size Max=4
    Uniform Sprite Size Min=1
    UsePositionOffset=false
    Write Material Random=true
   )]
  particleUpdateScript [
   &ParticleState()
   &ScaleSpriteSize(
    Scale Factor=&Multiply_Vector2D_ByFloat(
     Float=&FloatFromCurve(
      FloatCurve=DataInterface() {propertyValues {Curve {
         defaultValue=0
         keys [
          {arriveTangent=20.4672, interpMode=RCIM_Cubic, leaveTangent=20.4672, tangentMode=RCTM_User, time=0, value=1}
          {arriveTangent=-0.1734, interpMode=RCIM_Cubic, leaveTangent=-0.1734, time=0.0665, value=1.5318}
          {arriveTangent=-0.6829, interpMode=RCIM_Cubic, leaveTangent=-0.6829, tangentMode=RCTM_User, time=0.1452, value=0.9748}
          {interpMode=RCIM_Cubic, tangentMode=RCTM_User, time=1, value=0.2019}
         ]
        }}}
      Scale Curve=&RandomRangeFloat(Minimum=0.2, Maximum=0.3)
     )
     Vector2D=&MakeVector2D(
      X=&FloatFromCurve(FloatCurve=DataInterface() {propertyValues {Curve {keys [<at 0.5 1>,<at 0.9 0.5>]}}})
      Y=&FloatFromCurve(
       FloatCurve=DataInterface() {propertyValues {Curve {
          defaultValue=0
          keys [
           {interpMode=RCIM_Cubic, tangentMode=RCTM_User, time=0.28, value=1}
           {arriveTangent=1.8857, interpMode=RCIM_Cubic, leaveTangent=1.8857, time=0.5202, value=1.6}
           {arriveTangent=1.4857, interpMode=RCIM_Cubic, leaveTangent=4.6698, tangentMode=RCTM_Break, time=0.8103, value=2}
           <at 0.9473 3.4861 RCIM_Cubic>
          ]
         }}}
       Scale Curve=&RandomRangeFloat(Minimum=1, Maximum=1.4)
      )
     )
    )
    Scale Sprite Size Mode=::NewEnumerator2
   )
   &SkeletalMeshLocation(
    MaskSpawnGroup=false
    Mesh Sampling Type=::NewEnumerator5
    Sampled Position Offset=<0 0 5>
    Sampled Position Offset Coordinate Space:ENiagaraSkelMeshTransforms=::NewEnumerator2
    Skeletal Mesh=$System.left hand
    Socket Sampling Mode=::NewEnumerator5
   )
   &ShapeLocation(
    Additional Quaternion Rotation=true
    Additional Yaw / Pitch / Roll=&VectorFromCurve(VectorCurve=DataInterface() {propertyValues {
       XCurve {keys [<at 0 0>,<at 0.6279 0.1>,<at 0.931 0.488>]}
       YCurve {keys [<at 0 0>,<at 0.6279 0.1>,<at 0.931 0.488>,<at 0.931 0.488>]}
       ZCurve {keys [<at 0 0>,<at 0.6279 0.1>,<at 0.931 0.488>]}
      }})
    MaskSpawnGroup=false
    Offset=<0 0 5>
    Offset Mode=::NewEnumerator0
    Override Seed=false
    Rotation Mode=::NewEnumerator3
    Shape Origin=$Particles.SkeletalMeshLocation.SampledPosition
    Sphere Radius=&FloatFromCurve(
     FloatCurve=DataInterface() {propertyValues {Curve {keys [
         <at 2.4 26 RCIM_Cubic>
         {arriveTangent=31.5761, interpMode=RCIM_Cubic, leaveTangent=31.5761, time=3.1288, value=31.7586}
         {arriveTangent=229.2193, interpMode=RCIM_Cubic, leaveTangent=229.2193, time=3.2993, value=54.3952}
         <at 3.3809 89.536 RCIM_Cubic>
        ]}}}
     CurveIndex=$Emitter.Age
     Scale Curve=&RandomRangeFloat(Minimum=0.9, Maximum=1)
    )
    UseOwnerScale=true
   )
   &ScaleColor(
    Scale Alpha=&Multiply_Float(
     A=&FloatFromCurve(
      FloatCurve=DataInterface() {propertyValues {Curve {keys [<at 2.4 0>,<at 2.5 1>,<at 2.8 1>,<at 3.1516 0.43>,<at 3.6 0>]}}}
      CurveIndex=$Emitter.Age
     )
     B=&FloatFromCurve(
      FloatCurve=DataInterface() {propertyValues {Curve {
         defaultValue=1.0113
         keys [
          {interpMode=RCIM_Cubic, tangentMode=RCTM_User, time=0, value=0}
          {interpMode=RCIM_Cubic, tangentMode=RCTM_User, time=0.212, value=1}
          {tangentMode=RCTM_User, time=1, value=0}
         ]
        }}}
      Scale Curve=&FloatFromCurve(FloatCurve=DataInterface() {propertyValues {Curve {keys [
           <at 0 0.5 RCIM_Cubic>
           {arriveTangent=-0.1247, interpMode=RCIM_Cubic, leaveTangent=-0.1247, time=0.1795, value=1.0667}
           {arriveTangent=0.0434, interpMode=RCIM_Cubic, leaveTangent=0.0434, time=0.4011, value=0.45}
           {arriveTangent=0.1899, interpMode=RCIM_Cubic, leaveTangent=0.1899, time=0.5636, value=1.0834}
           {arriveTangent=0.206, interpMode=RCIM_Cubic, leaveTangent=0.206, time=0.7439, value=0.5151}
           {arriveTangent=-0.0589, interpMode=RCIM_Cubic, leaveTangent=-0.0589, time=0.8503, value=1.1424}
           <at 1 0.5 RCIM_Cubic>
          ]}}})
     )
    )
    Scale RGB=&MakeVectorFromLinearColorRGB(Color=&ColorFromCurve(ColorCurve=DataInterface() {propertyValues {
        AlphaCurve {keys []}
        BlueCurve {keys [
          <at 0 1>
          <at 0.1193 3.479>
          <at 0.1681 9>
          <at 0.214 5>
          {arriveTangent=-16.5932, interpMode=RCIM_Cubic, leaveTangent=-16.5932, time=0.4249, value=1}
          <at 0.455 1>
          <at 0.9238 0.3727 RCIM_Cubic>
         ]}
        GreenCurve {keys [
          <at 0 1>
          <at 0.1193 3.479>
          <at 0.1681 9>
          <at 0.214 5>
          {arriveTangent=-16.5932, interpMode=RCIM_Cubic, leaveTangent=-16.5932, time=0.4249, value=1}
          <at 0.455 1>
          <at 0.9238 0.6017 RCIM_Cubic>
         ]}
        RedCurve {keys [
          <at 0 1>
          <at 0.1193 3.479>
          <at 0.1681 9>
          <at 0.214 5>
          {arriveTangent=-16.5932, interpMode=RCIM_Cubic, leaveTangent=-16.5932, time=0.4249, value=1}
          <at 0.455 1>
          <at 0.9238 1 RCIM_Cubic>
         ]}
       }}))
    ScaleA=true
    ScaleRGB=true
   )
   &PointAttractionForce(
    Attraction Radius=120
    Attraction Strength=&Multiply_Float(
     A=&FloatFromCurve(
      FloatCurve=DataInterface() {propertyValues {Curve {
         defaultValue=1
         keys [<at 2.4 0>, {tangentMode=RCTM_User, time=2.9706, value=0.4943}, <at 3.2114 2.1364>, <at 3.4046 4.8273>, <at 3.492 7.8007>]
        }}}
      CurveIndex=$Emitter.Age
      Scale Curve=50
     )
     B=&FloatFromCurve(
      FloatCurve=DataInterface() {propertyValues {Curve {defaultValue=1, keys [{tangentMode=RCTM_User, time=0, value=0.5}, <at 0.8002 1>, {tangentMode=RCTM_User, time=1, value=1.25}]}}}
      Scale Curve=5
     )
    )
    Attractor Position:NiagaraPosition=&AddVectorToPosition(Position=$Particles.SkeletalMeshLocation.SampledPosition, Vector=<0 0 5>)
    Falloff Exponent=0
    Kill Within Radius=false
    Module.Use Falloff=false
    UseAttractorOffset=false
   )
   &SolveForcesAndVelocity(Clamp Velocity=false, Limit Acceleration=false)
   &ScaleSpriteSizeBySpeed(
    Max Scale Factor=&MakeVector2D(
     X=1.2
     Y=&FloatFromCurve(FloatCurve=DataInterface() {propertyValues {Curve {keys [<at 2.4 3>,<at 3.3352 3.7651>,<at 3.5242 4.8195>]}}}, CurveIndex=$Emitter.Age)
    )
    Min Scale Factor=<0.7 1.2>
   )
   /SetVariables(Particles.Vector 2D:Vector2f=&Multiply_Vector2D_ByFloat(Float=1.5, Vector2D=$Particles.SpriteSize)) {isSetParameters=true}
  ]
  rendererClasses [&NiagaraSpriteRendererProperties]
  renderers [
   {rendererClass=&NiagaraSpriteRendererProperties}
   {rendererClass=&NiagaraSpriteRendererProperties, rendererIndex=1}
  ]
 }
 Resurrect {
  emitterUpdateScript [
   &EmitterState(Life Cycle Mode=::NewEnumerator1, Loop Behavior=::NewEnumerator1, Loop Duration=5.1, UseLoopDelay=false)
   &SpawnBurst_Instantaneous(Spawn Time=2.57, Use Spawn Probability=false, UseLoopCountLimit=false)
   &SpawnBurst_Instantaneous(Spawn Group=1, Spawn Time=2.57, Use Spawn Probability=false, UseLoopCountLimit=false) {module=SpawnBurst_Instantaneous001}
   &SpawnBurst_Instantaneous(Spawn Group=2, Spawn Time=3.6, Use Spawn Probability=false, UseLoopCountLimit=false) {module=SpawnBurst_Instantaneous002}
   &SpawnBurst_Instantaneous(Spawn Count=3, Spawn Group=3, Spawn Time=3.6, Use Spawn Probability=false, UseLoopCountLimit=false) {module=SpawnBurst_Instantaneous003}
   &SpawnBurst_Instantaneous(Spawn Group=5, Spawn Time=2.4, Use Spawn Probability=false, UseLoopCountLimit=false) {module=SpawnBurst_Instantaneous005}
  ]
  particleSpawnScript [
   /SetVariables(Particles.VisibilityTag=&SpawnGroup()) {isSetParameters=true}
   &PartitionParticles(Number Of Partitions=/Niagara/DynamicInputs/Helpers/MaskInt(Mask Source:NiagaraInt32=0, Mask Parameter:NiagaraInt32=?, Mask:NiagaraInt32=3, Masked Int:NiagaraInt32=3, Passthrough Int:NiagaraInt32=0))
   &InitializeParticle(
    Color=&MergeHelper_ColorSwitch(
     0=<rgba 5 2.5958 0 1>
     1=<rgba 8 4.7412 0.4425 1>
     2=<rgba 1 1 1 1>
     3=<rgba 5 3.044 0.2976 1>
     4=<rgba 1 1 1 1>
     5=<rgba 1 1 1 1>
     Index=&SpawnGroup()
     Num=::NewEnumerator4
    )
    Lifetime=&MergeHelper_FloatSwitch(0=1.33, 1=1.345, 2=0.225, 3=0.25, 4=5.1, 5=1.6, Index=&SpawnGroup(), Num=::NewEnumerator4)
    Mass=&MaskFloat(
     Mask Source=0
     Mask Parameter=?
     Mask=3
     Masked Float=&RandomRangeFloat(Minimum=0.6, Maximum=1)
     Passthrough Float=1
    )
    Mass Mode=::NewEnumerator1
    Mesh Scale=&MergeHelper_VectorSwitch(
     0=&VectorFromFloat(Value=0.1)
     1=<0 0 0>
     2=&VectorFromFloat(Value=2.5)
     3=&VectorFromFloat(Value=&MergeHelper_FloatSwitch(0=5.2, 1=6, 2=6.8, 3=?, Index=$Particles.PartitionParticles.Partition, Num=::NewEnumerator1))
     4=<0 0 0>
     Index=&SpawnGroup()
     Num=::NewEnumerator3
    )
    Mesh Scale Mode=::NewEnumerator4
    Position Offset=&MaskVector(
     Mask Source=0
     Mask Parameter=?
     Mask=4
     Masked Vector=<0 5 0>
     Passthrough Vector=&MaskVector(Mask Source=0, Mask Parameter=?, Mask=5, Masked Vector=<0 70 150>, Passthrough Vector=<-10 0 0>)
    )
    Sprite Size=&MergeHelper_Vector2Switch(
     0=<0 0>
     1=&Vector2DFromFloat(Value=35)
     2=<0 0>
     3=<0 0>
     4=<0 0>
     Index=&SpawnGroup()
     Num=::NewEnumerator3
    )
    Sprite Size Mode=::NewEnumerator4
    UsePositionOffset=true
    Write Material Random=true
   )
   &CameraOffset(Camera Offset Amount=&MaskFloat(Mask Source=0, Mask Parameter=?, Mask=2, Masked Float=30, Passthrough Float=0))
   &InitialMeshOrientation(
    Mesh Axis To Orient=&MergeHelper_VectorSwitch(
     0=<1 0 0>
     1=<0 0 0>
     2=<1 0 0>
     3=&MergeHelper_VectorSwitch(
      0=&RandomRangeVector(Minimum=<0 -0.5 0.8>, Maximum=<1 -0.6 0.8>)
      1=&RandomRangeVector(Minimum=<0 -0.4 0.8>, Maximum=<1 -0.4 0.8>)
      2=&RandomRangeVector(Minimum=<0 -0.3 0.8>, Maximum=<1 -0.2 0.8>)
      3=?
      Index=$Particles.PartitionParticles.Partition
      Num=::NewEnumerator1
     )
     4=<0 0 0>
     Index=&SpawnGroup()
     Num=::NewEnumerator3
    )
    Mesh Orientation Mode=::NewEnumerator2
    Orientation Vector=&MergeHelper_VectorSwitch(
     0=<1 5 0>
     1=<0 0 0>
     2=<0 0 1>
     3=&RandomRangeVector(Minimum=<-0.5 -0.25 0>, Maximum=<0.5 0.25 0>)
     4=<0 0 0>
     Index=&SpawnGroup()
     Num=::NewEnumerator3
    )
    Rotation=&MergeHelper_VectorSwitch(
     0=<0 0 0>
     1=<0 0 0>
     2=&RandomRangeVector(Minimum=<0 0 0>, Maximum=<0 0 1>)
     3=<0 0 0>
     4=<0 0 0>
     Index=&SpawnGroup()
     Num=::NewEnumerator3
    )
    Use Rotation Vector=true
   )
  ]
  particleUpdateScript [
   &ParticleState()
   &ScaleSpriteSize(
    Scale Factor=&MergeHelper_Vector2Switch(
     0=<0 0>
     1=&Vector2DFromFloat(Value=&Multiply_Float(
       A=&FloatFromCurve(
        FloatCurve=DataInterface() {propertyValues {Curve {keys [
            {interpMode=RCIM_Cubic, leaveTangent=1.0643, tangentMode=RCTM_Break, time=2.57, value=0}
            {arriveTangent=0.7828, interpMode=RCIM_Cubic, leaveTangent=0.7828, tangentMode=RCTM_User, time=3.598, value=0.603}
            {arriveTangent=2.4635, interpMode=RCIM_Cubic, leaveTangent=2.4635, time=3.6, value=1.5568}
            {arriveTangent=1.0648, interpMode=RCIM_Cubic, leaveTangent=-2.3897, tangentMode=RCTM_Break, time=3.7348, value=0.9399}
            {arriveTangent=-5.2162, interpMode=RCIM_Cubic, leaveTangent=-5.2162, time=3.867, value=0.6554}
            <at 3.915 0 RCIM_Cubic>
           ]}}}
        CurveIndex=$Emitter.Age
        Scale Curve=5
       )
       B=0.3
      ))
     2=<0 0>
     3=<0 0>
     4=<0 0>
     Index=&SpawnGroup()
     Num=::NewEnumerator3
    )
    Scale Sprite Size Mode=::NewEnumerator2
   )
   &Timeline(
    Play Rate=&MaskFloat(
     Mask Source=0
     Mask Parameter=?
     Mask=0
     Masked Float=&FloatFromCurve(
      FloatCurve=DataInterface() {propertyValues {Curve {keys [
          <at 0.045 1.1639 RCIM_Cubic>
          {arriveTangent=0.1095, interpMode=RCIM_Cubic, leaveTangent=0.221, tangentMode=RCTM_Break, time=2.9796, value=1}
          {arriveTangent=0.9764, interpMode=RCIM_Cubic, leaveTangent=0.9764, time=3.3383, value=1.5253}
          <at 3.8 1.801 RCIM_Cubic>
         ]}}}
      CurveIndex=$Emitter.Age
     )
     Passthrough Float=1
    )
    Rewind Mode:ENiagara_TimelineMode=::NewEnumerator1
    Timeline End=9999999
   )
   &ScaleColor(
    Scale Alpha=&MergeHelper_FloatSwitch(
     0=&FloatFromCurve(FloatCurve=DataInterface() {propertyValues {Curve {
         defaultValue=1
         keys [
          {interpMode=RCIM_Cubic, tangentMode=RCTM_User, time=0, value=0}
          {arriveTangentWeight=0.1126, interpMode=RCIM_Cubic, leaveTangentWeight=0.1536, tangentMode=RCTM_User, time=0.3379, value=1}
          {arriveTangentWeight=0.2783, interpMode=RCIM_Cubic, leaveTangentWeight=0.1215, tangentMode=RCTM_User, time=0.7355, value=1}
          {tangentMode=RCTM_User, time=0.9447, value=-0.0188}
         ]
        }}})
     1=&FloatFromCurve(FloatCurve=DataInterface() {propertyValues {Curve {keys [
          {arriveTangent=2.5463, interpMode=RCIM_Cubic, leaveTangent=2.5463, tangentMode=RCTM_User, time=0, value=0}
          <at 0.2854 1>
          {arriveTangent=-1.3994, interpMode=RCIM_Cubic, leaveTangent=-1.3994, time=0.691, value=1}
          <at 1 0>
         ]}}})
     2=0
     3=&FloatFromCurve(FloatCurve=DataInterface() {propertyValues {Curve {keys [
          <at 0 0 RCIM_Cubic>
          {arriveTangent=1.1283, interpMode=RCIM_Cubic, leaveTangent=1.1283, time=0.4029, value=0.5366}
          {arriveTangentWeight=0.0575, interpMode=RCIM_Cubic, leaveTangentWeight=0.1382, tangentMode=RCTM_User, time=0.5868, value=0.6621}
          {arriveTangent=-1.6022, interpMode=RCIM_Cubic, leaveTangent=-1.6022, time=0.7506, value=0.4826}
          {arriveTangent=-0.0742, tangentMode=RCTM_User, time=1, value=0}
         ]}}})
     4=0
     Index=&SpawnGroup()
     Num=::NewEnumerator3
    )
    Scale RGB=&MergeHelper_VectorSwitch(0=<1 1 1>, 1=<1 1 1>, 2=<1 1 1>, 3=<1 1 1>, 4=<1 1 1>, Index=&SpawnGroup(), Num=::NewEnumerator3)
    ScaleA=true
    ScaleRGB=true
   )
   &SkeletalMeshLocation(
    MaskSpawnGroup=false
    Mesh Sampling Type=::NewEnumerator5
    Orientation Sampling=::NewEnumerator1
    Sampled Position Offset=&Add_Vector(
     A=&MergeHelper_VectorSwitch(0=<0 0 10>, 1=<0 5 8>, 2=<0 5 8>, 3=<0 5 8>, 4=<0 0 0>, Index=&SpawnGroup(), Num=::NewEnumerator3)
     B=&Lerp_Vector(
      A=<-2 -3 0>
      B=<-2 -3 12>
      Alpha=&ClampFloat(Float=&RemapRange(
        Input Value To Remap=&DotProduct(Vector A=&ReturnCameraVector(Camera Vectors=::NewEnumerator1, Coordinate Space=::World), Vector B=<0 0 -1>)
        Input Min=0.7
        Input Max=0.9
        Output Min=0
        Output Max=1
       ))
     )
    )
    Sampled Position Offset Coordinate Space:ENiagaraSkelMeshTransforms=::NewEnumerator2
    Skeletal Mesh=$System.left hand
    Socket Sampling Mode=::NewEnumerator5
   )
   &MeshRotationForce(
    Coordinate Space:ENiagaraMeshTransforms=::NewEnumerator2
    Rotation=&MaskVector(
     Mask Source=0
     Mask Parameter=?
     Mask=3
     Masked Vector=&MergeHelper_VectorSwitch(
      0=&MakeVector(
       X=&FloatFromCurve(
        FloatCurve=DataInterface() {propertyValues {Curve {defaultValue=1, keys [<at 0.2045 0 RCIM_Cubic>,<at 0.7755 1 RCIM_Cubic>]}}}
        Scale Curve=-15
       )
       Y=0
       Z=0
      )
      1=&MakeVector(
       X=&FloatFromCurve(
        FloatCurve=DataInterface() {propertyValues {Curve {defaultValue=1, keys [<at 0.2045 0 RCIM_Cubic>,<at 0.7755 1 RCIM_Cubic>]}}}
        Scale Curve=-25
       )
       Y=0
       Z=0
      )
      2=&MakeVector(
       X=&FloatFromCurve(
        FloatCurve=DataInterface() {propertyValues {Curve {defaultValue=1, keys [<at 0.2045 0 RCIM_Cubic>,<at 0.7755 1 RCIM_Cubic>]}}}
        Scale Curve=-45
       )
       Y=0
       Z=0
      )
      3=?
      Index=$Particles.PartitionParticles.Partition
      Num=::NewEnumerator1
     )
     Passthrough Vector=<0 0 0>
    )
   ) {module=MeshRotationForce001}
   &ScaleMeshSize(Scale Factor=&MergeHelper_VectorSwitch(
     0=&VectorFromFloat(Value=&FloatFromCurve(
       FloatCurve=DataInterface() {propertyValues {Curve {
          defaultValue=1
          keys [
           <at 0.0009 0.9236 RCIM_Cubic>
           {arriveTangent=1.5824, interpMode=RCIM_Cubic, leaveTangent=1.5824, time=0.394, value=1.4343}
           {arriveTangent=9.08, interpMode=RCIM_Cubic, leaveTangent=9.08, tangentMode=RCTM_User, time=0.7732, value=2.1457}
           {arriveTangent=153.0404, interpMode=RCIM_Cubic, leaveTangent=153.0404, time=0.778, value=5.6915}
           {arriveTangent=15.3438, interpMode=RCIM_Cubic, leaveTangent=15.3438, tangentMode=RCTM_User, time=0.7861, value=4.1086}
           {arriveTangent=-10.606, interpMode=RCIM_Cubic, leaveTangent=-10.606, time=0.7883, value=5.3304}
           {arriveTangent=-15.4668, interpMode=RCIM_Cubic, leaveTangent=-15.4668, tangentMode=RCTM_User, time=0.9082, value=2.8128}
          ]
         }}}
       Scale Curve=0.85
      ))
     1=<0 0 0>
     2=&VectorFromFloat(Value=&Multiply_Float(
       A=&FloatFromCurve(
        FloatCurve=DataInterface() {propertyValues {Curve {keys [
            {arriveTangent=21.0933, interpMode=RCIM_Cubic, leaveTangent=21.0933, tangentMode=RCTM_User, time=0, value=0.4}
            {arriveTangent=-78.4705, interpMode=RCIM_Cubic, leaveTangent=1.4963, tangentMode=RCTM_Break, time=0.0182, value=0.4652}
            {arriveTangent=0.5447, interpMode=RCIM_Cubic, leaveTangent=0.5447, time=0.2793, value=0.8179}
            {interpMode=RCIM_Cubic, tangentMode=RCTM_Break, time=1, value=1}
           ]}}}
        Scale Curve=0.75
       )
       B=0.27
      ))
     3=&VectorFromFloat(Value=&FloatFromCurve(
       FloatCurve=DataInterface() {propertyValues {Curve {keys [
           {arriveTangent=50.867, interpMode=RCIM_Cubic, leaveTangent=50.867, tangentMode=RCTM_User, time=0, value=0.3}
           {arriveTangent=-4.5493, interpMode=RCIM_Cubic, leaveTangent=1.4636, tangentMode=RCTM_Break, time=0.0724, value=0.6647}
           {arriveTangent=0.3615, interpMode=RCIM_Cubic, leaveTangent=0.3615, time=0.1802, value=0.7474}
           {arriveTangent=-0.142, interpMode=RCIM_Cubic, leaveTangent=-0.142, tangentMode=RCTM_User, time=1, value=1}
          ]}}}
       Scale Curve=0.33
      ))
     4=<0 0 0>
     Index=&SpawnGroup()
     Num=::NewEnumerator3
    ))
   /SetVariables(Particles.Energy Build Dynamic Param:Vector4f=&MakeVector4(X=$Particles.Timeline.Time, Y=0, Z=0, W=0)) {isSetParameters=true}
   /SetVariables(
    Particles.Flare Color:LinearColor=&ColorFromCurve(ColorCurve=DataInterface() {propertyValues {
       AlphaCurve {keys [<at 0.1655 0.75 RCIM_Cubic>,<at 1 0 RCIM_Cubic>]}
       BlueCurve {keys [<at 0 0 RCIM_Cubic>,<at 1 0.007 RCIM_Cubic>]}
       GreenCurve {keys [<at 0 1.8783 RCIM_Cubic>,<at 1 0.5251 RCIM_Cubic>]}
       RedCurve {keys [<at 0 3 RCIM_Cubic>,<at 1 1 RCIM_Cubic>]}
      }})
    Particles.Flare Size:Vector2f=&Vector2DFromFloat(Value=&Multiply_Float(
      A=250
      B=&FloatFromCurve(FloatCurve=DataInterface() {propertyValues {Curve {keys [
           {arriveTangent=16.2971, interpMode=RCIM_Cubic, leaveTangent=16.2971, tangentMode=RCTM_User, time=0, value=0}
           {arriveTangentWeight=0.0479, interpMode=RCIM_Cubic, leaveTangentWeight=0.1858, tangentMode=RCTM_User, time=0.05, value=1}
          ]}}})
     ))
   ) {isSetParameters=true}
   /SetVariables(Particles.Glow Color=&MakeLinearColorFromVectorAndFloat(
     "Vector (RGB)"=&MakeVectorFromLinearColorRGB(Color=<rgba 1 0.7399 0 1>)
     "Float (Alpha)"=&FloatFromCurve(FloatCurve=DataInterface() {propertyValues {Curve {keys [
          {arriveTangent=2.5463, interpMode=RCIM_Cubic, leaveTangent=2.5463, tangentMode=RCTM_User, time=0, value=0}
          <at 0.2854 0.5>
          {arriveTangent=-0.6997, interpMode=RCIM_Cubic, leaveTangent=-0.6997, time=0.691, value=0.5}
          <at 1 0>
         ]}}})
    )) {isSetParameters=true}
   &SolveForcesAndVelocity(Clamp Velocity=false, Limit Acceleration=false)
   /SetVariables(
    Particles.Energy Wave Color:LinearColor=&MultiplyLinearColor(
     Linear Color=&MultiplyLinearColor(
      Linear Color=$Particles.Initial.Color
      Scale Factor=&MakeVector4_FromLinearColor(LinearColor=&ColorFromCurve(ColorCurve=DataInterface() {propertyValues {
          AlphaCurve {keys [<at 0.1245 1>,<at 0.3677 1>,<at 1 0>]}
          BlueCurve {keys [<at 0 0.1574>,<at 1 0.1238>]}
          GreenCurve {keys [<at 0 0.685>,<at 1 0.3066>]}
          RedCurve {keys [<at 0 1>,<at 1 1>]}
         }}))
     )
     Scale Factor=&MakeVector4_FromLinearColor(LinearColor=&MakeLinearColorFromVectorAndFloat(
       "Vector (RGB)"=&VectorFromFloat(Value=&FloatFromCurve(FloatCurve=DataInterface() {propertyValues {Curve {keys [
             {interpMode=RCIM_Cubic, tangentMode=RCTM_Break, time=0, value=5}
             {arriveTangent=-0.0379, interpMode=RCIM_Cubic, leaveTangent=-0.0379, tangentMode=RCTM_User, time=0.15, value=2}
             {interpMode=RCIM_Cubic, tangentMode=RCTM_Break, time=1, value=1}
            ]}}}))
       "Float (Alpha)"=&FloatFromCurve(
        FloatCurve=DataInterface() {propertyValues {Curve {keys [
            <at -0.0003 0 RCIM_Cubic>
            {arriveTangent=1.9989, interpMode=RCIM_Cubic, leaveTangent=1.9989, time=0.2634, value=0.8066}
            {interpMode=RCIM_Cubic, tangentMode=RCTM_User, time=0.5, value=1}
            {interpMode=RCIM_Cubic, tangentMode=RCTM_User, time=1, value=0}
           ]}}}
        Scale Curve=0.3
       )
      ))
    )
    Particles.Energy Wave Dynamic Param:Vector4f=&MakeVector4(
     X=&FloatFromCurve(FloatCurve=DataInterface() {propertyValues {Curve {keys [{arriveTangent=-0.8686, interpMode=RCIM_Cubic, leaveTangent=-0.8686, tangentMode=RCTM_User, time=0, value=0.25}, <at 1 -0.15>]}}})
     Y=&FloatFromCurve(FloatCurve=DataInterface() {propertyValues {Curve {keys [{arriveTangent=-0.2232, interpMode=RCIM_Cubic, leaveTangent=-0.2232, tangentMode=RCTM_User, time=0, value=1}, <at 1 0.9>]}}})
     Z=&FloatFromCurve(FloatCurve=DataInterface() {propertyValues {Curve {keys [
          {interpMode=RCIM_Cubic, tangentMode=RCTM_Break, time=0, value=0.1}
          {arriveTangent=1.8128, interpMode=RCIM_Cubic, tangentMode=RCTM_Break, time=1, value=1.2}
         ]}}})
     W=0
    )
   ) {isSetParameters=true}
   /Niagara/Modules/Decal/Decal_Attributes(Decal Size:Vector3f=<20 600 600>, Fade=1, Size=1)
   &Light_Attributes(Diffuse Scale=1, Falloff Exponent=1, Radius=1, Specular Scale=1, Volumetric Scattering=1)
   /SetVariables(
    Particles.Decal Color:LinearColor=&ColorFromCurve(ColorCurve=DataInterface() {propertyValues {
       AlphaCurve {keys [<at 0.0326 0>,<at 0.3153 0.08>,<at 0.5549 0.055>,<at 0.8188 0.04>,<at 0.9722 0>]}
       BlueCurve {keys [<at 0.0908 0.3208>,<at 0.2884 0.6421>,<at 0.7248 0.3208>]}
       GreenCurve {keys [<at 0.0908 0.6412>,<at 0.2884 1.1183>,<at 0.7248 0.6412>]}
       RedCurve {keys [<at 0.0908 1>,<at 0.2884 1.5>,<at 0.7248 1>]}
      }})
    Particles.Front Light Color:LinearColor=&ColorFromCurve(ColorCurve=DataInterface() {propertyValues {
       AlphaCurve {keys [<at 0 0>,<at 0.2516 0.2>,<at 0.6803 0.7908>,<at 0.7684 1>,<at 0.8107 1>,<at 0.9053 0.25>,<at 0.9722 0>]}
       BlueCurve {keys [<at 0.6966 0.4812>,<at 0.7732 1.1253>,<at 1 0.4812>]}
       GreenCurve {keys [<at 0.6966 0.9618>,<at 0.7732 2.4976>,<at 1 0.9618>]}
       RedCurve {keys [<at 0.6966 1.5>,<at 0.7732 3.5>,<at 1 1.5>]}
      }})
   ) {isSetParameters=true}
  ]
  rendererClasses [
   &NiagaraMeshRendererProperties
   &NiagaraSpriteRendererProperties
   &NiagaraLightRendererProperties
  ]
  renderers [
   {rendererClass=&NiagaraMeshRendererProperties, rendererIndex=0}
   {rendererClass=&NiagaraSpriteRendererProperties, rendererIndex=1}
   {rendererClass=&NiagaraSpriteRendererProperties, rendererIndex=2}
   {rendererClass=&NiagaraSpriteRendererProperties, rendererIndex=3}
   {rendererClass=&NiagaraMeshRendererProperties, rendererIndex=4}
   {rendererClass=&NiagaraMeshRendererProperties, rendererIndex=5}
   {rendererClass=&NiagaraLightRendererProperties, rendererIndex=6}
  ]
 }
 Spark_Burst001 {rendererClasses [], renderers []}
 Spiral {rendererClasses [], renderers []}
}
summary {
 systemName=NS_Emote_TrackWaver
 emitters [
  {emitterName=Joint_Glows, rendererClasses [&NiagaraSpriteRendererProperties]}
  {emitterName=Ambient_Hand_Sparks, rendererClasses [&NiagaraSpriteRendererProperties]}
  {emitterName=Motion_Trails, rendererClasses [&NiagaraRibbonRendererProperties]}
  {emitterName=Spiral, rendererClasses []}
  {emitterName=Ground_Glow, rendererClasses []}
  {emitterName=Spark_Burst001, rendererClasses []}
  {emitterName=Res_BuildUp_Sparks002, rendererClasses [&NiagaraSpriteRendererProperties]}
  {
   emitterName=ConesMistCondensed
   rendererClasses [
    &NiagaraMeshRendererProperties
    &NiagaraSpriteRendererProperties
    &NiagaraLightRendererProperties
   ]
  }
  {
   emitterName=Resurrect
   rendererClasses [
    &NiagaraMeshRendererProperties
    &NiagaraSpriteRendererProperties
    &NiagaraLightRendererProperties
   ]
  }
 ]
 userVariables [
  User.bIsFrontEnd {type=NiagaraBool}
  User.bIsFrontEndPreview {type=NiagaraBool}
  User.IsLargeBody {type=NiagaraBool}
 ]
}
