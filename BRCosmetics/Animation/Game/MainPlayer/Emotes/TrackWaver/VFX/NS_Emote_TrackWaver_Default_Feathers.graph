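# Niagara system for the default feathers VFX of the TrackWaver player emote, showing ten mesh feathers over about four seconds.
# A burst spawns one particle per partition, and the partition index picks mesh scale, one of two mirrored orientations, and a feather socket.
# Particles follow their skeletal mesh sockets each update.
# Two multiplied curves fade alpha, and a time-of-day tint uses the same warm color for every period.
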NiagaraSystem /BRCosmetics/Animation/Game/MainPlayer/Emotes/TrackWaver/VFX/NS_Emote_TrackWaver_Default_Feathers
dependencies {
 usedDataInterfaces [&NiagaraDataInterfaceArrayInt32, &NiagaraDataInterfaceCurve, &NiagaraDataInterfaceSkeletalMesh]
 usedDynamicInputs []
 usedModules [
  &self:SystemSpawnScript.SystemScriptSource.SystemScriptGraph.NiagaraNodeAssignment_1.SetVariables_#1
  &SystemState
 ]
}
emitters {Feathers {
  emitterUpdateScript [
   &EmitterState(Life Cycle Mode=::NewEnumerator1, Loop Behavior=::NewEnumerator1, Loop Duration=4.2, UseLoopDelay=false)
   &SpawnBurst_Instantaneous(Spawn Count=10, Use Spawn Probability=false, UseLoopCountLimit=false)
  ]
  particleSpawnScript [
   &PartitionParticles(Number Of Partitions=10)
   &InitializeParticle(
    Color Mode=::NewEnumerator0
    Lifetime=4.2
    Mesh Scale=&MergeHelper_VectorSwitch(
     0=<0.075 0.12 0.1>
     1=<0.085 0.13 0.1>
     2=<0.105 0.13 0.1>
     3=<0.12 0.13 0.1>
     4=<0.135 0.115 0.1>
     5=<0.075 0.12 0.1>
     6=<0.085 0.13 0.1>
     7=<0.105 0.13 0.1>
     8=<0.12 0.13 0.1>
     9=<0.135 0.115 0.1>
     Index=$Particles.PartitionParticles.Partition
     Num=::NewEnumerator8
    )
    Mesh Scale Mode=::NewEnumerator4
    UsePositionOffset=false
    Write Material Random=true
   )
   &InitialMeshOrientation(
    Mesh Orientation Mode=::NewEnumerator2
    Orientation Vector=&MergeHelper_VectorSwitch(
     0=<-1 0 0>
     1=<-1 0 0>
     2=<-1 0 0>
     3=<-1 0 0>
     4=<-1 0 0>
     5=<1 0 0>
     6=<1 0 0>
     7=<1 0 0>
     8=<1 0 0>
     9=<1 0 0>
     Index=$Particles.PartitionParticles.Partition
     Num=::NewEnumerator8
    )
    Use Rotation Vector=false
   )
   &CameraOffset(Camera Offset Amount=-1)
  ]
  particleUpdateScript [
   &ParticleState()
   &SkeletalMeshLocation(
    Bone / Socket Index=&SelectIntFromArray(
     Direct Array Index=$Particles.PartitionParticles.Partition
     Int Selection Array=DataInterface() {propertyValues {IntData [0,1,2,3,4,5,6,7,8,9]}}
    )
    MaskSpawnGroup=false
    Mesh Sampling Type=::NewEnumerator5
    Skeletal Mesh=$System.FeatherSockets
    Socket Sampling Mode=::NewEnumerator5
   )
   &ScaleColor(
    Scale Alpha=&Multiply_Float(
     A=&FloatFromCurve(
      FloatCurve=DataInterface() {propertyValues {Curve {keys [<at 0.1491 -0.0094>,<at 0.2946 0.3548>,<at 0.8283 1>,<at 1.7432 1>,<at 2.3867 0.6126>,<at 3.0085 0.4182>,<at 3.2468 0.2333>,<at 3.7572 0.1446>,<at 4.0648 0.0095>]}}}
      CurveIndex=$Emitter.Age
     )
     B=&FloatFromCurve(
      FloatCurve=DataInterface() {propertyValues {Curve {
         defaultValue=1
         keys [
          {interpMode=RCIM_Cubic, tangentMode=RCTM_User, time=0.0024, value=0.4}
          {interpMode=RCIM_Cubic, tangentMode=RCTM_User, time=0.4205, value=1.0156}
          {arriveTangent=-1.0667, interpMode=RCIM_Cubic, leaveTangent=-1.0667, time=0.6782, value=0.9452}
          {tangentMode=RCTM_User, time=0.9976, value=0.4}
         ]
        }}}
      Scale Curve=1.35
     )
    )
    Scale RGB=&MakeVectorFromLinearColorRGB(Color=&FNTimeOfDayColor(
      Night Color=<rgba 2 1.562 0.7349 1>
      Morning Color=<rgba 2 1.562 0.7349 1>
      Day Color=<rgba 2 1.562 0.7349 1>
      Evening Color=<rgba 2 1.562 0.7349 1>
     ))
    ScaleA=true
    ScaleRGB=true
   )
  ]
  rendererClasses [&NiagaraMeshRendererProperties]
  renderers [{rendererClass=&NiagaraMeshRendererProperties, rendererIndex=0}]
 }}
summary {
 systemName=NS_Emote_TrackWaver_Default_Feathers
 emitters [{emitterName=Feathers, rendererClasses [&NiagaraMeshRendererProperties]}]
 userVariables []
}
