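# Niagara effect for the TrackWaver emote, named for the wings despawn, that plays a short gold-orange burst.
# The Burst emitter fires four instantaneous spawn groups, using the group index as visibility tag and selector for color, lifetime, size, offset and alpha fade.
# Three sprite renderers and a mesh renderer use curve-driven scale and color, with platform and view-angle-dependent camera offset.
# The two Spiral emitters show no renderers.
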NiagaraSystem /BRCosmetics/Animation/Game/MainPlayer/Emotes/TrackWaver/VFX/NS_Emote_TrackWaver_WingsDespawn
dependencies {
 usedDataInterfaces [
  &NiagaraDataInterfacePlatformSet
  &NiagaraDataInterfaceVectorCurve
  &NiagaraDataInterfaceCamera
  &NiagaraDataInterfaceCurve
  &NiagaraDataInterfaceVector2DCurve
  &NiagaraDataInterfaceColorCurve
  &NiagaraDataInterfaceSkeletalMesh
 ]
 usedDynamicInputs []
 usedModules [
  &self:SystemSpawnScript.SystemScriptSource.SystemScriptGraph.NiagaraNodeAssignment_0.SetVariables_#1
  &self:SystemSpawnScript.SystemScriptSource.SystemScriptGraph.NiagaraNodeAssignment_1.SetVariables_#2
  &SystemState
 ]
}
emitters {
 Burst {
  emitterUpdateScript [
   &EmitterState(Life Cycle Mode=::NewEnumerator1, Loop Behavior=::NewEnumerator1, Loop Duration=2, UseLoopDelay=false)
   &SpawnBurst_Instantaneous(Spawn Time=0.02, Use Spawn Probability=false, UseLoopCountLimit=false) {module=SpawnBurst_Instantaneous002}
   &SpawnBurst_Instantaneous(Spawn Group=2, Spawn Time=0.02, Use Spawn Probability=false, UseLoopCountLimit=false) {module=SpawnBurst_Instantaneous001}
   &SpawnBurst_Instantaneous(Spawn Group=3, Use Spawn Probability=false, UseLoopCountLimit=false) {module=SpawnBurst_Instantaneous003}
   &SpawnBurst_Instantaneous(Spawn Group=4, Use Spawn Probability=false, UseLoopCountLimit=false) {module=SpawnBurst_Instantaneous004}
  ]
  particleSpawnScript [
   /SetVariables(Particles.VisibilityTag=&SpawnGroup()) {isSetParameters=true}
   &InitializeParticle(
    Color=&MergeHelper_ColorSwitch(
     0=<rgba 2 0.9352 0.2272 1.2>
     1=<rgba 1 0.316 0.2009 0.7>
     2=<rgba 2.5 0.9171 0 1.25>
     3=<rgba 1 1 1 1>
     4=<rgba 3 1.2738 0 0.8522>
     Index=&SpawnGroup()
     Num=::NewEnumerator3
    )
    Lifetime=&MergeHelper_FloatSwitch(0=0.5, 1=0.5, 2=0.7, 3=0.83, 4=0.25, Index=&SpawnGroup(), Num=::NewEnumerator3)
    Mesh Scale Mode=::NewEnumerator3
    Mesh Uniform Scale=1.2
    Position Offset=&MergeHelper_VectorSwitch(0=<0 0 -3>, 1=<0 0 -3>, 2=<0 0 -5>, 3=<0 0 5>, 4=<0 0 0>, Index=&SpawnGroup(), Num=::NewEnumerator3)
    Sprite Rotation Angle=90
    Sprite Rotation Mode=::NewEnumerator0
    Sprite Size=&MergeHelper_Vector2Switch(0=<160 260>, 1=<160 260>, 2=<0 0>, 3=<250 250>, 4=<200 200>, Index=&SpawnGroup(), Num=::NewEnumerator3)
    Sprite Size Mode=::NewEnumerator4
    UsePositionOffset=true
    Write Material Random=true
   )
   &CameraOffset(Camera Offset Amount=&MaskFloat(
     Mask Source=0
     Mask Parameter=?
     Mask=2
     Masked Float=&MakeCustomFloatFromBool(
      Bool=&SetBoolByFloatComparison(
       A=&FN_PickFloatDependOnPlatform(
        TargetPlatform=DataInterface() {propertyValues {Platforms {
           qualityLevelMask=-29
           deviceProfileStates [
            {profileName=Linux, qualityLevelMask=1, setQualityLevelMask=3}
            {profileName=Mac, qualityLevelMask=1, setQualityLevelMask=3}
            {profileName=Mobile, qualityLevelMask=4294967295, setQualityLevelMask=4294967295}
            {profileName=Windows, qualityLevelMask=1, setQualityLevelMask=3}
            {profileName=Switch, qualityLevelMask=4294967295, setQualityLevelMask=4294967295}
           ]
          }}}
        TargetPlatformActive=1
        TargetPlatformNoActive=0
       )
       B=0
      )
      True Float=&RemapRange(
       Input Value To Remap=&MakeFloatFromVector(
        Vector=&VectorFromCurve(
         VectorCurve=DataInterface() {propertyValues {XCurve {keys [<at 0 1>,<at 1 0>]}, YCurve {keys [<at 0 1>,<at 1 0>]}, ZCurve {keys [<at 0 1>,<at 1 0>]}}}
         CurveIndex=&DotProduct(
          Vector A=&ReturnCameraVector(
           Camera Vectors=::NewEnumerator4
           Coordinate Space=::Simulation
           Origin Position=$Particles.Position
           CameraQuery=DataInterface() {propertyValues {bRequireCurrentFrameData=false}}
          )
          Vector B=<0 1 0>
         )
        )
        Channel=::NewEnumerator0
       )
       Input Min=0.2
       Input Max=0.9
       Output Min=-15
       Output Max=20
      )
      False Float=10
     )
     Passthrough Float=10
    ))
  ]
  particleUpdateScript [
   &ParticleState()
   &ScaleMeshSize(Scale Factor=&VectorFromCurve(
     VectorCurve=DataInterface() {propertyValues {
       XCurve {
        defaultValue=0
        keys [
         {arriveTangent=7.2229, interpMode=RCIM_Cubic, leaveTangent=7.2229, tangentMode=RCTM_User, time=0, value=0}
         {interpMode=RCIM_Cubic, tangentMode=RCTM_User, time=0.1515, value=0.7519}
         {interpMode=RCIM_Cubic, tangentMode=RCTM_User, time=0.9849, value=1.05}
        ]
       }
       YCurve {keys [
         {arriveTangent=8.0619, interpMode=RCIM_Cubic, leaveTangent=8.0619, leaveTangentWeight=0.3333, tangentMode=RCTM_User, time=0, value=0}
         {arriveTangentWeight=0.069, interpMode=RCIM_Cubic, leaveTangentWeight=0.2643, tangentMode=RCTM_User, time=0.1584, value=0.6945}
         {arriveTangentWeight=0.2556, interpMode=RCIM_Cubic, leaveTangentWeight=0.0087, tangentMode=RCTM_User, time=0.988, value=0}
        ]}
       ZCurve {
        defaultValue=0
        keys [
         {arriveTangent=9.0638, interpMode=RCIM_Cubic, leaveTangent=9.0638, tangentMode=RCTM_User, time=0, value=0}
         {arriveTangentWeight=0.069, interpMode=RCIM_Cubic, tangentMode=RCTM_User, time=0.141, value=0.7268}
         {arriveTangentWeight=0.2593, interpMode=RCIM_Cubic, tangentMode=RCTM_User, time=0.9849, value=0}
        ]
       }
      }}
     Scale Curve=<1.3 1.3 1.3>
    ))
   &ScaleSpriteSize(
    Scale Factor=/Niagara/DynamicInputs/Helpers/MaskVector2d(
     Mask Source:NiagaraInt32=0
     Mask Parameter:NiagaraInt32=?
     Mask:NiagaraInt32=4
     Masked Vector2D:Vector2f=<1 1>
     Passthrough Vector2D:Vector2f=&Vector2DFromCurve(
      Vector2Curve=DataInterface() {propertyValues {
        XCurve {
         defaultValue=0
         keys [
          {interpMode=RCIM_Cubic, leaveTangent=4.0442, tangentMode=RCTM_Break, time=0, value=0.1}
          {arriveTangent=4.8062, interpMode=RCIM_Cubic, leaveTangent=4.8062, time=0.1027, value=0.4785}
          {arriveTangentWeight=0.0254, interpMode=RCIM_Cubic, tangentMode=RCTM_User, time=0.1788, value=0.9593}
          {
           arriveTangent=-0.7572
           arriveTangentWeight=0.0475
           interpMode=RCIM_Cubic
           leaveTangent=-0.7572
           tangentMode=RCTM_User
           time=0.3212
           value=0.2892
          }
          <at 0.9979 0.0156>
         ]
        }
        YCurve {
         defaultValue=0.1
         keys [
          {arriveTangent=41.408, interpMode=RCIM_Cubic, leaveTangent=41.408, tangentMode=RCTM_User, time=0, value=0.1}
          {arriveTangentWeight=0.0152, interpMode=RCIM_Cubic, tangentMode=RCTM_User, time=0.0457, value=0.9462}
          {
           arriveTangent=-11.5805
           arriveTangentWeight=0.0195
           interpMode=RCIM_Cubic
           leaveTangent=-0.9291
           tangentMode=RCTM_Break
           time=0.1042
           value=0.4729
          }
          {arriveTangent=-0.1541, arriveTangentWeight=0.2986, interpMode=RCIM_Cubic, leaveTangent=-0.1541, tangentMode=RCTM_User, time=1, value=0}
         ]
        }
       }}
      ScaleCurve=<1.3 0.75>
     )
    )
    Scale Sprite Size Mode=::NewEnumerator2
   )
   &ScaleColor(
    Scale Alpha=&MergeHelper_FloatSwitch(
     0=&FloatFromCurve(FloatCurve=DataInterface() {propertyValues {Curve {keys [
          <at -0.0046 1.0113 RCIM_Cubic>
          {arriveTangent=-1.0036, interpMode=RCIM_Cubic, leaveTangent=-1.0036, time=0.5003, value=0.2765}
          <at 1.0011 0.002 RCIM_Cubic>
         ]}}})
     1=&FloatFromCurve(FloatCurve=DataInterface() {propertyValues {Curve {keys [
          <at -0.0046 1.0113 RCIM_Cubic>
          {arriveTangent=-1.0036, interpMode=RCIM_Cubic, leaveTangent=-1.0036, time=0.5003, value=0.2765}
          <at 1.0011 0.002 RCIM_Cubic>
         ]}}})
     2=&FloatFromCurve(FloatCurve=DataInterface() {propertyValues {Curve {keys [
          <at -0.0046 1.0113 RCIM_Cubic>
          {arriveTangent=-1.0036, interpMode=RCIM_Cubic, leaveTangent=-1.0036, time=0.5003, value=0.2765}
          <at 1.0011 0.002 RCIM_Cubic>
         ]}}})
     3=&FloatFromCurve(FloatCurve=DataInterface() {propertyValues {Curve {keys [
          <at -0.0046 1.0113 RCIM_Cubic>
          {arriveTangent=-1.0036, interpMode=RCIM_Cubic, leaveTangent=-1.0036, time=0.5003, value=0.2765}
          <at 1.0011 0.002 RCIM_Cubic>
         ]}}})
     4=&FloatFromCurve(FloatCurve=DataInterface() {propertyValues {Curve {
         defaultValue=1
         keys [
          {interpMode=RCIM_Cubic, tangentMode=RCTM_User, time=0, value=0}
          {interpMode=RCIM_Cubic, tangentMode=RCTM_User, time=0.2, value=0.85}
          {arriveTangent=-1.0625, interpMode=RCIM_Cubic, leaveTangent=-1.0625, time=0.43, value=0.29}
          {tangentMode=RCTM_User, time=1, value=0}
         ]
        }}})
     Index=&SpawnGroup()
     Num=::NewEnumerator3
    )
    ScaleA=true
    ScaleRGB=false
   )
   &SolveForcesAndVelocity(Clamp Velocity=false, Limit Acceleration=false)
   /SetVariables(Particles.Mesh Scale Color:LinearColor=&ColorFromCurve(ColorCurve=DataInterface() {propertyValues {
       AlphaCurve {keys [<at 0.2748 0.25 RCIM_Cubic>,<at 0.5629 0.12>,<at 1 0 RCIM_Cubic>]}
       BlueCurve {keys [
         <at 0 0.0526 RCIM_Cubic>
         <at 0.1114 0.0703>
         <at 0.3448 0.1023>
         <at 0.6276 0>
         {arriveTangent=0.2881, interpMode=RCIM_Cubic, leaveTangent=0.2881, time=0.7762, value=0.0577}
         <at 0.945 0.0914>
        ]}
       GreenCurve {keys [
         <at 0 0.3385 RCIM_Cubic>
         <at 0.1114 1.3119>
         <at 0.3448 1.0502>
         <at 0.6276 0.7348>
         {arriveTangent=-1.5371, interpMode=RCIM_Cubic, leaveTangent=-1.5371, time=0.7762, value=0.4125}
         <at 0.945 0.247>
        ]}
       RedCurve {keys [
         <at 0 1 RCIM_Cubic>
         <at 0.1114 2>
         <at 0.3448 2>
         <at 0.6276 1.2845>
         {arriveTangent=-0.8963, interpMode=RCIM_Cubic, leaveTangent=-0.8963, time=0.7762, value=1}
         <at 0.945 1>
        ]}
      }})) {isSetParameters=true}
   /SetVariables(
    Particles.Corona Scale Color:LinearColor=&MakeLinearColorFromVectorAndFloat(
     "Vector (RGB)"=&MakeVectorFromLinearColorRGB(Color=<rgba 1.5 0.6641 0.0573 2>)
     "Float (Alpha)"=&FloatFromCurve(
      FloatCurve=DataInterface() {propertyValues {Curve {
         defaultValue=1
         keys [
          {arriveTangent=20.985, interpMode=RCIM_Cubic, leaveTangent=20.985, tangentMode=RCTM_User, time=0, value=0}
          {arriveTangent=-3.9091, interpMode=RCIM_Cubic, leaveTangent=-3.9091, tangentMode=RCTM_User, time=0.0828, value=1.004}
          {arriveTangent=-1.5525, interpMode=RCIM_Cubic, leaveTangent=-1.5525, tangentMode=RCTM_User, time=0.2526, value=0.4508}
          {arriveTangent=-0.0405, interpMode=RCIM_Cubic, leaveTangent=-0.0405, tangentMode=RCTM_User, time=0.6224, value=0}
         ]
        }}}
      Scale Curve=0.15
     )
    )
    Particles.Corona Scale Sprite Size:Vector2f=&Multiply_Vector2D(
     A=$Particles.Initial.SpriteSize
     B=&Vector2DFromFloat(Value=&FloatFromCurve(FloatCurve=DataInterface() {propertyValues {Curve {
          defaultValue=0
          keys [
           <at 0 1 RCIM_Cubic>
           {
            arriveTangent=-10.7434
            arriveTangentWeight=0.4775
            interpMode=RCIM_Cubic
            leaveTangent=-0.0819
            leaveTangentWeight=0.1681
            tangentMode=RCTM_Break
            time=0.0912
            value=0.4955
           }
           {arriveTangent=0.1458, interpMode=RCIM_Cubic, leaveTangent=0.1458, tangentMode=RCTM_User, time=0.7179, value=0}
          ]
         }}}))
    )
   ) {isSetParameters=true}
   /SetVariables(Particles.Airglow Dynamic Param:Vector4f=&MakeVector4(X=2, Y=0, Z=0, W=0)) {isSetParameters=true}
  ]
  rendererClasses [&NiagaraSpriteRendererProperties, &NiagaraMeshRendererProperties]
  renderers [
   {rendererClass=&NiagaraSpriteRendererProperties}
   {rendererClass=&NiagaraMeshRendererProperties, rendererIndex=1}
   {rendererClass=&NiagaraSpriteRendererProperties, rendererIndex=2}
   {rendererClass=&NiagaraSpriteRendererProperties, rendererIndex=3}
  ]
 }
 Spiral {rendererClasses [], renderers []}
 Spiral001 {rendererClasses [], renderers []}
}
summary {
 systemName=NS_Emote_TrackWaver_WingsDespawn
 emitters [
  {emitterName=Spiral, rendererClasses []}
  {emitterName=Spiral001, rendererClasses []}
  {emitterName=Burst, rendererClasses [&NiagaraSpriteRendererProperties, &NiagaraMeshRendererProperties]}
 ]
 userVariables [
  User.bIsFrontEnd {type=NiagaraBool}
  User.bIsFrontEndPreview {type=NiagaraBool}
 ]
}
