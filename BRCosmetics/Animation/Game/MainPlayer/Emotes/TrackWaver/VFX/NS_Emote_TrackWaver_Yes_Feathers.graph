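# Niagara system for the TrackWaver "Yes" emote that shows twelve mesh feathers attached to skeletal mesh sockets.
# Three burst groups are partitioned per particle to pick each feather's socket, scale and orientation.
# The spawn group also sets a visibility tag alongside three mesh renderers.
# Each update re-samples socket locations, fades alpha with two multiplied curves, one driven by emitter age, and applies a time-of-day color.
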NiagaraSystem /BRCosmetics/Animation/Game/MainPlayer/Emotes/TrackWaver/VFX/NS_Emote_TrackWaver_Yes_Feathers
dependencies {
 usedDataInterfaces [&NiagaraDataInterfaceArrayInt32, &NiagaraDataInterfaceCurve, &NiagaraDataInterfaceSkeletalMesh]
 usedDynamicInputs []
 usedModules [
  &self:SystemSpawnScript.SystemScriptSource.SystemScriptGraph.NiagaraNodeAssignment_0.SetVariables_#1
  &SystemState
 ]
}
emitters {Feathers {
  emitterUpdateScript [
   &EmitterState(Life Cycle Mode=::NewEnumerator1, Loop Behavior=::NewEnumerator1, Loop Duration=4.2, UseLoopDelay=false)
   &SpawnBurst_Instantaneous(Spawn Count=2, Use Spawn Probability=false, UseLoopCountLimit=false)
   &SpawnBurst_Instantaneous(Spawn Count=2, Spawn Group=1, Use Spawn Probability=false, UseLoopCountLimit=false) {module=SpawnBurst_Instantaneous001}
   &SpawnBurst_Instantaneous(Spawn Count=8, Spawn Group=2, Use Spawn Probability=false, UseLoopCountLimit=false) {module=SpawnBurst_Instantaneous002}
  ]
  particleSpawnScript [
   /SetVariables(Particles.VisibilityTag=&SpawnGroup()) {isSetParameters=true}
   &PartitionParticles(Number Of Partitions=&MergeHelper_IntSwitch(0=2, 1=2, 2=8, 3=?, 4=?, 5=?, Index=&SpawnGroup(), Num=::NewEnumerator1))
   &InitializeParticle(
    Color Mode=::NewEnumerator0
    Lifetime=5
    Mesh Scale=&MergeHelper_VectorSwitch(
     0=<0.065 0.075 0.1>
     1=<0.2 0.21 0.1>
     2=&MergeHelper_VectorSwitch(
      0=<0.125 0.125 0.1>
      1=<0.125 0.125 0.1>
      2=<0.125 0.15 0.1>
      3=<0.15 0.17 0.1>
      4=<0.125 0.125 0.1>
      5=<0.125 0.125 0.1>
      6=<0.125 0.15 0.1>
      7=<0.15 0.17 0.1>
      Index=$Particles.PartitionParticles.Partition
      Num=::NewEnumerator6
     )
     3=?
     Index=&SpawnGroup()
     Num=::NewEnumerator1
    )
    Mesh Scale Mode=::NewEnumerator4
    UsePositionOffset=false
    Write Material Random=true
   )
   &InitialMeshOrientation(
    Mesh Orientation Mode=::NewEnumerator2
    Orientation Vector=&MergeHelper_VectorSwitch(
     0=&MergeHelper_VectorSwitch(0=<-1 0 0>, 1=<1 0 0>, 2=?, 3=?, Index=$Particles.PartitionParticles.Partition, Num=::NewEnumerator0)
     1=&MergeHelper_VectorSwitch(0=<-0.75 0.5 0>, 1=<0.75 -0.5 0>, 2=?, 3=?, Index=$Particles.PartitionParticles.Partition, Num=::NewEnumerator0)
     2=&MergeHelper_VectorSwitch(
      0=<-1 0.08 0>
      1=<-1 0.05 0>
      2=<-1 -0.1 0>
      3=<-1 -0.2 0>
      4=<1 -0.1 0>
      5=<1 -0.01 0>
      6=<1 0.08 0>
      7=<1 0.22 0>
      Index=$Particles.PartitionParticles.Partition
      Num=::NewEnumerator6
     )
     3=?
     Index=&SpawnGroup()
     Num=::NewEnumerator1
    )
    Use Rotation Vector=false
   )
   &CameraOffset(Camera Offset Amount=-1)
  ]
  particleUpdateScript [
   &ParticleState()
   &SkeletalMeshLocation(
    Bone / Socket Index=&MergeHelper_IntSwitch(
     0=&SelectIntFromArray(Direct Array Index=$Particles.PartitionParticles.Partition, Int Selection Array=DataInterface() {propertyValues {IntData [4,9]}})
     1=&SelectIntFromArray(Direct Array Index=$Particles.PartitionParticles.Partition, Int Selection Array=DataInterface() {propertyValues {IntData [10,11]}})
     2=&SelectIntFromArray(Direct Array Index=$Particles.PartitionParticles.Partition, Int Selection Array=DataInterface() {propertyValues {IntData [0,1,2,3,5,6,7,8]}})
     3=?
     4=?
     5=?
     Index=&SpawnGroup()
     Num=::NewEnumerator1
    )
    MaskSpawnGroup=false
    Mesh Sampling Type=::NewEnumerator5
    Skeletal Mesh=$System.FeatherSockets
    Socket Sampling Mode=::NewEnumerator5
   )
   &ScaleColor(
    Scale Alpha=&Multiply_Float(
     A=&FloatFromCurve(
      FloatCurve=DataInterface() {propertyValues {Curve {keys [<at 0.1491 -0.0094>,<at 0.2946 0.3548>,<at 0.8283 1>,<at 1.7432 1>,<at 2.3867 0.6126>,<at 3.0085 0.4182>,<at 3.2468 0.2333>,<at 3.7572 0.1446>,<at 4.0648 0.0095>]}}}
      CurveIndex=$Emitter.Age
     )
     B=&FloatFromCurve(
      FloatCurve=DataInterface() {propertyValues {Curve {
         defaultValue=1
         keys [
          {interpMode=RCIM_Cubic, tangentMode=RCTM_User, time=0.0024, value=0.4}
          {interpMode=RCIM_Cubic, tangentMode=RCTM_User, time=0.4205, value=1.0156}
          {arriveTangent=-1.0667, interpMode=RCIM_Cubic, leaveTangent=-1.0667, time=0.6782, value=0.9452}
          {tangentMode=RCTM_User, time=0.9976, value=0.4}
         ]
        }}}
      Scale Curve=1.35
     )
    )
    Scale RGB=&MakeVectorFromLinearColorRGB(Color=&FNTimeOfDayColor(Night Color=<rgba 15 5 0 1>, Morning Color=<rgba 15 5 0 1>, Day Color=<rgba 15 5 0 1>, Evening Color=<rgba 15 5 0 1>))
    ScaleA=true
    ScaleRGB=true
   )
  ]
  rendererClasses [&NiagaraMeshRendererProperties]
  renderers [
   {rendererClass=&NiagaraMeshRendererProperties, rendererIndex=0}
   {rendererClass=&NiagaraMeshRendererProperties, rendererIndex=1}
   {rendererClass=&NiagaraMeshRendererProperties, rendererIndex=2}
  ]
 }}
summary {
 systemName=NS_Emote_TrackWaver_Yes_Feathers
 emitters [{emitterName=Feathers, rendererClasses [&NiagaraMeshRendererProperties]}]
 userVariables []
}
